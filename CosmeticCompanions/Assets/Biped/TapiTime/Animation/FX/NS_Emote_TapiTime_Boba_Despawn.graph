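# Niagara system for the TapiTime companion's boba despawn emote effect: a one-shot sprite burst of pops and flying droplets.
# Pop spawns three groups, tagged for its three sprite renderers, outward from a system burst position, with the masked group smaller and faster.
# Spawn throws droplets in a cone from HeartIntroSpawnPosition under gravity and drag.
# Both use speed-stretched sprites and curve-driven material parameters; the three LWE emitters have no renderers.
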NiagaraSystem /CosmeticCompanions/Assets/Biped/TapiTime/Animation/FX/NS_Emote_TapiTime_Boba_Despawn
dependencies {
 usedDataInterfaces [
  &NiagaraDataInterfaceSpriteRendererInfo
  &NiagaraDataInterfaceCurve
  &NiagaraDataInterfaceVectorCurve
  &NiagaraDataInterfaceArrayColor
 ]
 usedDynamicInputs [&SelectLinearColorFromArray]
 usedModules [
  &self:SystemSpawnScript.SystemScriptSource.SystemScriptGraph.NiagaraNodeAssignment_0.SetVariables_#1
  &self:SystemSpawnScript.SystemScriptSource.SystemScriptGraph.NiagaraNodeAssignment_1.SetVariables_#2
  &SystemState
 ]
}
emitters {
 AirglowLWE {rendererClasses [], renderers []}
 Pop {
  emitterUpdateScript [
   &EmitterState()
   &SpawnBurst_Instantaneous(Spawn Count=3, Use Spawn Probability=false, UseLoopCountLimit=false)
   &SpawnBurst_Instantaneous(Spawn Count=3, Spawn Group=1, Use Spawn Probability=false, UseLoopCountLimit=false) {module=SpawnBurst_Instantaneous002}
   &SpawnBurst_Instantaneous(Spawn Count=15, Spawn Group=2, Use Spawn Probability=false, UseLoopCountLimit=false) {module=SpawnBurst_Instantaneous001}
  ]
  particleSpawnScript [
   &PartitionParticles()
   /SetVariables(Particles.VisibilityTag=&SpawnGroup()) {isSetParameters=true}
   &InitializeParticle(
    Color=<rgba 0.491 0.491 0.491 1>
    Lifetime=&MaskFloat(
     Mask Source=0
     Mask Parameter=?
     Mask=2
     Masked Float=&RandomRangeFloat(Minimum=0.4, Maximum=0.7)
     Passthrough Float=&RandomRangeFloat(Minimum=0.25, Maximum=0.5)
    )
    Sprite Rotation Mode=::NewEnumerator2
    Sprite Size Mode=::NewEnumerator3
    Uniform Sprite Size=&MaskFloat(
     Mask Source=0
     Mask Parameter=?
     Mask=2
     Masked Float=&RandomRangeFloat(Minimum=1, Maximum=1.5)
     Passthrough Float=&RandomRangeFloat(Minimum=8, Maximum=15)
    )
    UsePositionOffset=false
    Write Material Random=true
   )
   /SetVariables(Particles.DropletSize:Vector2f=&Multiply_Vector2D(A=$Particles.SpriteSize, B=<1 3>)) {isSetParameters=true}
   &DynamicMaterialParameters(
    Index 0 Param 4=1
    Module.Index0Float0Write=true
    Module.Index0Float1Write=true
    Module.Index0Float2Write=true
    Module.Index0Float3Write=true
   )
   &SubUVAnimation(
    Evaluation Type=::NewEnumerator0
    "Sprite Renderer "=DataInterface() {propertyValues {SpriteRenderer=&self:Pop_0.NiagaraSpriteRendererProperties_1}}
    SubUV Animation Mode:ENiagara_SubUVLookupModeV2=::NewEnumerator2
    UseEndFrame=false
    UseStartFrame=false
   )
   &ShapeLocation(MaskSpawnGroup=false, Override Seed=false, Sphere Radius=5, UseOwnerScale=true)
   &AddVelocity(
    Module.Use Falloff Distance=false
    Origin Offset=<0 0 -5>
    Origin Offset Coordinate Space=::World
    Override Seed=false
    Velocity Mode=::NewEnumerator1
    Velocity Origin:Vector3f=$System.BurstPosition
    Velocity Speed=&MaskFloat(
     Mask Source=0
     Mask Parameter=?
     Mask=2
     Masked Float=&RandomRangeFloat(Minimum=150, Maximum=300)
     Passthrough Float=&RandomRangeFloat(Minimum=75, Maximum=100)
    )
   )
  ]
  particleUpdateScript [
   &ParticleState()
   &ScaleSpriteSize(
    Scale Sprite Size Mode=::NewEnumerator0
    Uniform Scale Factor=&FloatFromCurve(FloatCurve=DataInterface() {propertyValues {Curve {keys [
         {arriveTangent=5.7724, interpMode=RCIM_Cubic, leaveTangent=5.7724, tangentMode=RCTM_User, time=0, value=0.4}
         {arriveTangent=0.788, tangentMode=RCTM_User, time=1, value=4}
        ]}}})
   ) {module=ScaleSpriteSize002}
   &ScaleColor(
    Scale Alpha=&FloatFromCurve(FloatCurve=DataInterface() {propertyValues {Curve {
        defaultValue=0
        keys [
         {arriveTangent=0.3137, interpMode=RCIM_Cubic, leaveTangent=0.3137, tangentMode=RCTM_User, time=0, value=1}
         {arriveTangent=-2.0958, interpMode=RCIM_Cubic, leaveTangent=-2.0958, tangentMode=RCTM_User, time=1, value=0}
        ]
       }}})
    Scale RGB=&VectorFromCurve(VectorCurve=DataInterface() {propertyValues {
       XCurve {keys [
         <at 0 0 RCIM_Cubic>
         {interpMode=RCIM_Cubic, tangentMode=RCTM_User, time=0.2, value=2}
         {
          arriveTangent=0.0281
          arriveTangentWeight=0.6561
          interpMode=RCIM_Cubic
          leaveTangent=0.0281
          leaveTangentWeight=0.6561
          tangentMode=RCTM_User
          tangentWeightMode=RCTWM_WeightedBoth
          time=1
          value=0.3
         }
        ]}
       YCurve {keys [
         <at 0 0 RCIM_Cubic>
         {interpMode=RCIM_Cubic, tangentMode=RCTM_User, time=0.2, value=2}
         {
          arriveTangent=0.0281
          arriveTangentWeight=0.6561
          interpMode=RCIM_Cubic
          leaveTangent=0.0281
          leaveTangentWeight=0.6561
          tangentMode=RCTM_User
          tangentWeightMode=RCTWM_WeightedBoth
          time=1
          value=0.3
         }
        ]}
       ZCurve {keys [
         <at 0 0 RCIM_Cubic>
         {interpMode=RCIM_Cubic, tangentMode=RCTM_User, time=0.2, value=2}
         {
          arriveTangent=0.0281
          arriveTangentWeight=0.6561
          interpMode=RCIM_Cubic
          leaveTangent=0.0281
          leaveTangentWeight=0.6561
          tangentMode=RCTM_User
          tangentWeightMode=RCTWM_WeightedBoth
          time=1
          value=0.3
         }
        ]}
      }})
    ScaleA=true
    ScaleRGB=true
   )
   &AccelerationForce(Acceleration=&MaskVectorBySpawnGroup(Masked Vector=<0 0 -980>, Passthrough Vector=<1 1 -490>, Spawn Group Mask=2))
   &SolveForcesAndVelocity(Clamp Velocity=false, Limit Acceleration=false)
   &ScaleSpriteSizeBySpeed(Max Scale Factor=<2 5>)
   /SetVariables(Particles.Glow Color=&Multiply_LinearColorByFloat(
     Linear Color=&MultiplyLinearColor(
      Linear Color=$Particles.Color
      Scale Factor=&MakeVector4_FromLinearColor(LinearColor=<rgba 1 0.5877 0.8563 1>)
     )
     Float=0.33
    )) {isSetParameters=true}
   &DynamicMaterialParameters(
    Index 0 Param 2=&FloatFromCurve(FloatCurve=DataInterface() {propertyValues {Curve {defaultValue=1, keys [{interpMode=RCIM_Cubic, tangentMode=RCTM_User, time=0, value=0}, {tangentMode=RCTM_User, time=1, value=1}]}}})
    Index 0 Param 3=&FloatFromCurve(FloatCurve=DataInterface() {propertyValues {Curve {keys [{interpMode=RCIM_Cubic, tangentMode=RCTM_User, time=0, value=0.3749}, <at 1 0>]}}})
    Module.Index0Float0Write=true
    Module.Index0Float1Write=true
    Module.Index0Float2Write=true
    Module.Index0Float3Write=true
   ) {module=DynamicMaterialParameters001}
  ]
  rendererClasses [&NiagaraSpriteRendererProperties]
  renderers [
   {rendererClass=&NiagaraSpriteRendererProperties}
   {rendererClass=&NiagaraSpriteRendererProperties, rendererIndex=1}
   {rendererClass=&NiagaraSpriteRendererProperties, rendererIndex=2}
  ]
 }
 SmokeLWE {rendererClasses [], renderers []}
 SparklesLWE {rendererClasses [], renderers []}
 Spawn {
  emitterUpdateScript [
   &EmitterState()
   &SpawnBurst_Instantaneous(Spawn Count=11, Use Spawn Probability=false, UseLoopCountLimit=false)
  ]
  particleSpawnScript [
   &InitializeParticle(
    AdjustAlpha=false
    AdjustHue=false
    AdjustSaturation=false
    AdjustValue=true
    Color=<rgba 0.8 0.8 0.8 1>
    Color Mode=::NewEnumerator3
    Lifetime=&RandomRangeFloat(Minimum=0.4, Maximum=0.5)
    Sprite Size Mode=::NewEnumerator3
    Uniform Sprite Size=&RandomRangeFloat(Minimum=1.5, Maximum=3)
    UsePositionOffset=false
    Value Range:Vector2f=<0.4 1>
    Write Material Random=true
   )
   &ShapeLocation(
    MaskSpawnGroup=false
    Non Uniform Scale=<1 1 5>
    Offset=$System.HeartIntroSpawnPosition
    Offset Coordinate Space=::Local
    Offset Mode=::NewEnumerator0
    Override Seed=false
    Sphere Radius=0.5
    UseOwnerScale=true
   )
   &AddVelocity(
    Cone Angle=120
    Cone Axis=<0 0 1>
    Module.UseConeFalloff=false
    Module.UseVelDistribution=false
    Override Seed=false
    Velocity Mode=::NewEnumerator2
    Velocity Speed=&RandomRangeFloat(Minimum=200, Maximum=400)
   )
   &ApplyInitialForces()
   &CameraOffset(Camera Offset Amount=5)
   &SubUVAnimation(Evaluation Type=::NewEnumerator0, SubUV Animation Mode:ENiagara_SubUVLookupModeV2=::NewEnumerator2, UseEndFrame=false, UseStartFrame=false)
  ]
  particleUpdateScript [
   &AccelerationForce(Acceleration=<0 0 -980>)
   &Drag(Use Linear Drag=true, Use Rotational Drag=false)
   &SolveForcesAndVelocity(Clamp Velocity=false, Limit Acceleration=false)
   &ParticleState()
   &ScaleSpriteSizeBySpeed(Max Scale Factor=<2 5>)
   &DynamicMaterialParameters(
    Index 0 Param 2=&FloatFromCurve(FloatCurve=DataInterface() {propertyValues {Curve {defaultValue=1, keys [{interpMode=RCIM_Cubic, tangentMode=RCTM_User, time=0, value=0}, {tangentMode=RCTM_User, time=1, value=1}]}}})
    Index 0 Param 3=&FloatFromCurve(FloatCurve=DataInterface() {propertyValues {Curve {keys [{interpMode=RCIM_Cubic, tangentMode=RCTM_User, time=0, value=0.3749}, <at 1 0>]}}})
    Module.Index0Float0Write=true
    Module.Index0Float1Write=true
    Module.Index0Float2Write=false
    Module.Index0Float3Write=true
   ) {module=DynamicMaterialParameters001}
  ]
  rendererClasses [&NiagaraSpriteRendererProperties]
  renderers [{rendererClass=&NiagaraSpriteRendererProperties}]
 }
}
summary {
 systemName=NS_Emote_TapiTime_Boba_Despawn
 emitters [
  {emitterName=SparklesLWE, rendererClasses []}
  {emitterName=AirglowLWE, rendererClasses []}
  {emitterName=SmokeLWE, rendererClasses []}
  {emitterName=Pop, rendererClasses [&NiagaraSpriteRendererProperties]}
  {emitterName=Spawn, rendererClasses [&NiagaraSpriteRendererProperties]}
 ]
 userVariables [User.Variant {type=NiagaraInt32, defaultValue=-1}]
}
