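# Niagara system for the boba trails in the TapiTime cosmetic companion emote, a short-lived effect with two emitters.
# A ribbon emitter lays down a thin, short trail that inherits velocity, falls under drag and fades through dynamic material parameters.
# A sprite emitter adds a burst plus stream of small flecks, thrown in a cone and stretched by speed.
# A Variant user integer and color-array selection at system spawn suggest per-variant tinting.
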NiagaraSystem /CosmeticCompanions/Assets/Biped/TapiTime/Animation/FX/NS_Emote_TapiTime_Boba_Trails
dependencies {
 usedDataInterfaces [
  &NiagaraDataInterfaceCurve
  &NiagaraDataInterfaceSpriteRendererInfo
  &NiagaraDataInterfaceArrayColor
 ]
 usedDynamicInputs [&SelectLinearColorFromArray, &Multiply_LinearColorByFloat]
 usedModules [
  &self:SystemSpawnScript.SystemScriptSource.SystemScriptGraph.NiagaraNodeAssignment_0.SetVariables_#1
  &self:SystemSpawnScript.SystemScriptSource.SystemScriptGraph.NiagaraNodeAssignment_1.SetVariables_#2
  &SystemState
 ]
}
emitters {
 Minimal {
  emitterUpdateScript [
   &EmitterState(Life Cycle Mode=::NewEnumerator1, Loop Behavior=::NewEnumerator1, Loop Duration=0.5, UseLoopDelay=false)
   &SpawnRate(SpawnRate=50, Use Spawn Probability=false)
  ]
  particleSpawnScript [
   &InitializeParticle(
    Color Mode=::NewEnumerator0
    Lifetime=0.2
    Ribbon Width=&RandomRangeFloat(Minimum=5, Maximum=10)
    Ribbon Width Mode=::NewEnumerator1
    UsePositionOffset=false
    Write Material Random=true
   )
   &AddVelocity(Velocity=<0 0 500>) {enabled=false}
   &InheritVelocity(Inherited Velocity Speed Limit=80, Speed Limit=true, UseSpeedThreshold=false)
   &ApplyInitialForces()
  ]
  particleUpdateScript [
   &ParticleState()
   &DynamicMaterialParameters(
    Index 0 Param 1=&MaskFloatBySpawnGroup(Masked Float=2.2288, Passthrough Float=18.6053, Spawn Group Mask=1)
    Index 0 Param 2=&FloatFromCurve(FloatCurve=DataInterface() {propertyValues {Curve {keys [<at 0 1 RCIM_Cubic>, {arriveTangent=0.0167, interpMode=RCIM_Cubic, leaveTangent=0.0167, tangentMode=RCTM_Break, time=0.8, value=0}]}}})
    Index 0 Param 3=1
    Index 0 Param 4=1
    Module.Index0Float0Write=true
    Module.Index0Float1Write=true
    Module.Index0Float2Write=true
    Module.Index0Float3Write=true
   ) {module=DynamicMaterialParameters002}
   &ScaleColor(Scale Alpha=0.3, ScaleA=true, ScaleRGB=true)
   &AccelerationForce(Acceleration=<0 0 -1200>)
   &Drag(Drag=3, Use Linear Drag=true, Use Rotational Drag=false)
   &SolveForcesAndVelocity(Clamp Velocity=false, Limit Acceleration=false)
  ]
  rendererClasses [&NiagaraRibbonRendererProperties]
  renderers [{rendererClass=&NiagaraRibbonRendererProperties}]
 }
 Minimal001 {
  emitterUpdateScript [
   &EmitterState(Life Cycle Mode=::NewEnumerator1, Loop Behavior=::NewEnumerator1, Loop Duration=0.7, UseLoopDelay=false)
   &SpawnRate(SpawnRate=50, Use Spawn Probability=false)
   &SpawnBurst_Instantaneous(Spawn Count=11, Use Spawn Probability=false, UseLoopCountLimit=false)
  ]
  particleSpawnScript [
   &InitializeParticle(
    AdjustAlpha=false
    AdjustHue=false
    AdjustSaturation=false
    AdjustValue=true
    Color=<rgba 0.8 0.8 0.8 1>
    Color Mode=::NewEnumerator3
    Lifetime=&RandomRangeFloat(Minimum=0.4, Maximum=0.6)
    Sprite Size=&Multiply_Vector2D_ByFloat(
     Float=&RandomRangeFloat(Minimum=0.7, Maximum=1.5)
     Vector2D=&RandomRangeVector2D(Minimum=<1 1.5>, Maximum=<1.5 2>)
    )
    Sprite Size Mode=::NewEnumerator4
    UsePositionOffset=false
    Value Range:Vector2f=<0.4 1>
    Write Material Random=true
   )
   &ShapeLocation(MaskSpawnGroup=false, Non Uniform Scale=<1 1 5>, Offset Mode=::NewEnumerator0, Override Seed=false, Sphere Radius=0.5, UseOwnerScale=true)
   &AddVelocity(
    Cone Angle=90
    Cone Axis=<0 0 1>
    Module.UseConeFalloff=false
    Module.UseVelDistribution=false
    Override Seed=false
    Velocity Mode=::NewEnumerator2
    Velocity Speed=&RandomRangeFloat(Minimum=80, Maximum=150)
   )
   &CameraOffset(Camera Offset Amount=25)
   &SubUVAnimation(Evaluation Type=::NewEnumerator0, SubUV Animation Mode:ENiagara_SubUVLookupModeV2=::NewEnumerator2, UseEndFrame=false, UseStartFrame=false)
   &InheritVelocity(Speed Limit=true, UseSpeedThreshold=false)
   &ApplyInitialForces()
  ]
  particleUpdateScript [
   &AccelerationForce(Acceleration=<0 0 -980>)
   &Drag(Use Linear Drag=true, Use Rotational Drag=false)
   &SolveForcesAndVelocity(Clamp Velocity=false, Limit Acceleration=false)
   &ParticleState()
   &ScaleSpriteSizeBySpeed(Max Scale Factor=<2 10>)
   &DynamicMaterialParameters(
    Index 0 Param 1=&RandomRangeFloat(Minimum=0, Maximum=1)
    Index 0 Param 2=&FloatFromCurve(FloatCurve=DataInterface() {propertyValues {Curve {defaultValue=1, keys [{interpMode=RCIM_Cubic, tangentMode=RCTM_User, time=0, value=0}, {tangentMode=RCTM_User, time=1, value=1}]}}})
    Index 0 Param 3=&FloatFromCurve(FloatCurve=DataInterface() {propertyValues {Curve {keys [{interpMode=RCIM_Cubic, tangentMode=RCTM_User, time=0, value=0.3749}, <at 1 0>]}}})
    Module.Index0Float0Write=true
    Module.Index0Float1Write=true
    Module.Index0Float2Write=false
    Module.Index0Float3Write=true
   ) {module=DynamicMaterialParameters001}
  ]
  rendererClasses [&NiagaraSpriteRendererProperties]
  renderers [{rendererClass=&NiagaraSpriteRendererProperties}]
 }
}
summary {
 systemName=NS_Emote_TapiTime_Boba_Trails
 emitters [
  {emitterName=Minimal, rendererClasses [&NiagaraRibbonRendererProperties]}
  {emitterName=Minimal001, rendererClasses [&NiagaraSpriteRendererProperties]}
 ]
 userVariables [User.Variant {type=NiagaraInt32, defaultValue=-1}]
}
