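# Niagara system for the TapiTime emote's stage spotlights, with colors, positions and pulsing driven by system beat, bar, oscillation and flicker values.
# SpotLightMain is an overhead spotlight built from meshes, sprites, light and component renderers, and a floor decal.
# SpotLightSecondary spawns three partitioned beams that index user position and color arrays per bar, orient along a fixed-point axis and vary brightness per beat.
# User LightsFade dims both, and front-end offsets adjust placement.
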NiagaraSystem /CosmeticCompanions/Assets/Biped/TapiTime/Animation/FX/NS_Emote_TapiTime_SpotLights
dependencies {
 usedDataInterfaces [&NiagaraDataInterfaceArrayFloat, &NiagaraDataInterfaceArrayInt32, &NiagaraDataInterfaceCurve]
 usedDynamicInputs [
  &Divide_Float
  &FloatFromCurve
  /Niagara/DynamicInputs/Divide/Divide_Int
  &ModuloInt
 ]
 usedModules [
  &self:SystemSpawnScript.SystemScriptSource.SystemScriptGraph.NiagaraNodeAssignment_4.SetVariables_#1
  &SystemState
  &self:SystemSpawnScript.SystemScriptSource.SystemScriptGraph.NiagaraNodeAssignment_2.SetVariables_#2
  &self:SystemSpawnScript.SystemScriptSource.SystemScriptGraph.NiagaraNodeAssignment_1.SetVariables_#3
  &self:SystemSpawnScript.SystemScriptSource.SystemScriptGraph.NiagaraNodeAssignment_0.SetVariables_#4
 ]
}
emitters {
 SpotLightMain {
  emitterUpdateScript [
   &EmitterState(
    Inactive Response=::NewEnumerator1
    Life Cycle Mode=::NewEnumerator1
    Loop Behavior=::NewEnumerator1
    Loop Duration Mode=::NewEnumerator1
    UseLoopDelay=false
   )
   &SpawnBurst_Instantaneous(Use Spawn Probability=false, UseLoopCountLimit=false)
  ]
  particleSpawnScript [
   &InitializeParticle(
    Mesh Scale=<0.6 0.6 0.7>
    Mesh Scale Mode=::NewEnumerator4
    Sprite Size Mode=::NewEnumerator3
    Uniform Sprite Size=200
    UsePositionOffset=false
    Write Material Random=true
   )
   &CameraOffset()
   /SetVariables(
    Particles.FloorGlowFacingVector:Vector3f=<0 0 1>
    Particles.FloorGlowPosition:NiagaraPosition=&AddVectorToPosition(Position=/Niagara/DynamicInputs/Helpers/SimulationPosition(), Vector=<0 0 4>)
    Particles.FloorGlowSize:Vector2f=&Vector2DFromFloat(Value=150)
    Particles.LightAttenuationScale=130
    Particles.LightDiffuseScale=50
    Particles.LightExponent=20
    Particles.LightPosition:NiagaraPosition=&AddVectorToPosition(Position=&SimulationPosition(), Vector=<0 0 150>)
    Particles.LightPosition HighOnly:NiagaraPosition=&AddVectorToPosition(
     Position=&SimulationPosition()
     Vector=&AddVectorToPosition(Position=<0 0 75>, Vector=Linked() {linkedVariable {name=System.FEPosition, type=NiagaraPosition}})
    )
    Particles.LightRadius=50
    Particles.LightSpecularScale=50
    Particles.SpotLightPosition:NiagaraPosition=&AddVectorToPosition(
     Position=&SimulationPosition()
     Vector=&AddVectorToPosition(Position=<0 0 150>, Vector=Linked() {linkedVariable {name=System.FEPosition, type=NiagaraPosition}})
    )
   ) {isSetParameters=true, module=SetVariables_#5}
  ]
  particleUpdateScript [
   &ParticleState(
    Kill Particles When Lifetime Has Elapsed=false
    Let Infinitely Lived Particles Die When Emitter Deactivates=true
    Loop Particles Lifetime=true
   )
   &Color(
    Color=&SelectLinearColorFromArray(
     Array Sampling Mode=::NewEnumerator1
     Direct Array Index=$System.Bar
     Direct Array Mode=::NewEnumerator1
     Color Selection Array=$User.SpotlightColorArray
     Random Seed=?
    )
    Scale Alpha=&Multiply_Float(
     A=&Multiply_Float(
      A=$System.Oscillation
      B=&SelectFloatFromArray(
       Direct Array Index=$System.LoopCount
       Direct Array Mode=::NewEnumerator1
       Float Selection Array=DataInterface() {propertyValues {FloatData [0.4,0.4,0.4,0.4,0.4]}}
      )
     )
     B=$User.LightsFade
    )
   )
   /SetVariables(
    Particles.CoreSpotlightColor:LinearColor=&MultiplyLinearColor(
     Linear Color=$Particles.Color
     Scale Factor=&MakeFloat4FromVectorAndFloat("Vector (RGB)"=<1 1 1>, "Float (Alpha)"=$System.Flicker)
    )
    Particles.FlareColor=&MultiplyLinearColor(Linear Color=$Particles.Color, Scale Factor=&MakeVector4_FromFloat(Float=0.4))
    Particles.FloorGlowColor:LinearColor=&MultiplyLinearColor(Linear Color=$Particles.Color, Scale Factor=&MakeVector4_FromFloat(Float=0.666))
    Particles.LightColor=&MultiplyLinearColor(Linear Color=$Particles.Color, Scale Factor=&MakeVector4_FromFloat(Float=8))
    Particles.LightComponentIntensity:NiagaraFloat=&Multiply_Float(A=$System.Oscillation, B=7111)
   ) {isSetParameters=true}
   /Niagara/Modules/Decal/Decal_Attributes(
    Color=1
    Color Value To Scale:LinearColor=$Particles.Color
    Decal Facing Mode=1
    Decal Size:Vector3f=<50 300 300>
    Orientation=1
    Position=1
    Scale Alpha=0.3
    Scale RGB:Vector3f=<0.7 0.7 0.7>
    ScaleA=true
    ScaleRGB=true
    Set Alignment=false
    Size=1
    UsePositionOffset=true
    Visible=1
   )
  ]
  rendererClasses [
   &NiagaraMeshRendererProperties
   &NiagaraLightRendererProperties
   &NiagaraComponentRendererProperties
   &NiagaraSpriteRendererProperties
   &NiagaraDecalRendererProperties
  ]
  renderers [
   {rendererClass=&NiagaraMeshRendererProperties, rendererIndex=0}
   {rendererClass=&NiagaraMeshRendererProperties, rendererIndex=1}
   {rendererClass=&NiagaraLightRendererProperties, rendererIndex=2}
   {rendererClass=&NiagaraComponentRendererProperties, rendererIndex=3}
   {rendererClass=&NiagaraMeshRendererProperties, rendererIndex=4}
   {rendererClass=&NiagaraMeshRendererProperties, rendererIndex=5}
   {rendererClass=&NiagaraSpriteRendererProperties, rendererIndex=6}
   {rendererClass=&NiagaraSpriteRendererProperties, rendererIndex=7}
   {rendererClass=&NiagaraDecalRendererProperties, rendererIndex=8}
  ]
 }
 SpotLightSecondary {
  emitterUpdateScript [
   &EmitterState(
    Inactive Response=::NewEnumerator1
    Life Cycle Mode=::NewEnumerator1
    Loop Behavior=::NewEnumerator1
    Loop Duration Mode=::NewEnumerator1
    UseLoopDelay=false
   )
   &SpawnBurst_Instantaneous(Spawn Count=3, Use Spawn Probability=false, UseLoopCountLimit=false)
  ]
  particleSpawnScript [
   &InitializeParticle(
    Color=<rgba 0.3 0.271 0.24 1>
    Mesh Scale=<0.5 0.5 0.8>
    Mesh Scale Mode=::NewEnumerator4
    Sprite Size Mode=::NewEnumerator3
    Uniform Sprite Size=111
    UsePositionOffset=false
    Write Material Random=true
   )
   &CameraOffset()
   &PartitionParticles(Number Of Partitions=3)
  ]
  particleUpdateScript [
   &ParticleState(
    Kill Particles When Lifetime Has Elapsed=false
    Let Infinitely Lived Particles Die When Emitter Deactivates=true
    Loop Particles Lifetime=true
   )
   /SetVariables(Particles.Position=&AddVectorToPosition(
     Position=&SelectVectorFromArray(
      Direct Array Index=&Add_Integer(A=$Particles.PartitionParticles.Partition, B=&Multiply_Int(A=$System.Bar, B=3))
      Direct Array Mode=::NewEnumerator1
      Vector Selection Array=$User.SpotlightsPositionArray
     )
     Vector=&FN_FrontEndVector3Modifier(Gameplay=<0 0 0>, Front End=<0 0 150>)
    )) {isSetParameters=true, module=SetVariables_#5}
   &ScaleMeshSize(Scale Factor=&SelectVectorFromArray(
     Direct Array Index=&Add_Integer(A=$Particles.PartitionParticles.Partition, B=&Multiply_Int(A=$System.Bar, B=3))
     Direct Array Mode=::NewEnumerator1
     Vector Selection Array=DataInterface() {propertyValues {FloatData [<1 1 1>, <1 1 1>, <1 1 1>, <1 1 1>, <1 1 1>, <1 1 1>, <1 1 0.8>, <1 1 0.8>, <1 1 0.8>, <1 1 1>, <1 1 1>, <1 1 1>]}}
    ))
   &UpdateMeshOrientation(
    Facing Direction=&NormalizeVector(
     Vector To Normalize=&Subtract_Vector(
      A=&ConvertPositionToVector(Input Position=$Particles.Position, Conversion Operation:ENiagara_LWCConvertPosToVec=::NewEnumerator0)
      B=<0 0 18.6807>
     )
     Safe Normalize=true
     Safe Fallback Threshold=0
     Safe Fallback Vector=<0 0 0.001>
    )
    Facing Interpolation=false
    Orientation Method=::NewEnumerator1
    Override Quat=false
    Reference Axis Vector:Vector3f=<0 0 1>
    Stabilization Mode:ENiagara_OrientVectorCount=::NewEnumerator0
    TransformFacing=false
   )
   &Color(
    Color=&Multiply_LinearColorByFloat(
     Linear Color=&SelectLinearColorFromArray(
      Array Sampling Mode=::NewEnumerator1
      Direct Array Index=&SelectIntFromArray(
       Direct Array Index=$System.Bar
       Direct Array Mode=::NewEnumerator1
       Int Selection Array=DataInterface() {propertyValues {IntData [3,2,4,0,1]}}
      )
      Direct Array Mode=::NewEnumerator1
      Color Selection Array=$User.SpotlightColorArray
      Random Seed=?
     )
     Float=$User.LightsFade
    )
    Scale Alpha=&Multiply_Float(
     A=$System.Flicker
     B=&SelectFloatFromArray(
      Direct Array Index=&Add_Integer(A=&Multiply_Int(A=$System.Beat, B=3), B=$Particles.PartitionParticles.Partition)
      Direct Array Mode=::NewEnumerator1
      Float Selection Array=DataInterface() {propertyValues {FloatData [0.6,1,0.4,1,0.4,0.6,0.4,0.6,1]}}
     )
    )
    Scale Color=<0.4 0.4 0.4>
   )
   /SetVariables(
    Particles.GlowCameraOffset=11
    Particles.GlowColor=&MultiplyLinearColor(Linear Color=$Particles.Color, Scale Factor=&MakeVector4_FromFloat(Float=0.8))
   ) {isSetParameters=true}
  ]
  rendererClasses [&NiagaraMeshRendererProperties, &NiagaraSpriteRendererProperties]
  renderers [
   {rendererClass=&NiagaraMeshRendererProperties, rendererIndex=0}
   {rendererClass=&NiagaraMeshRendererProperties, rendererIndex=1}
   {rendererClass=&NiagaraSpriteRendererProperties, rendererIndex=2}
  ]
 }
}
summary {
 systemName=NS_Emote_TapiTime_SpotLights
 emitters [
  {
   emitterName=SpotLightMain
   rendererClasses [
    &NiagaraMeshRendererProperties
    &NiagaraLightRendererProperties
    &NiagaraComponentRendererProperties
    &NiagaraSpriteRendererProperties
    &NiagaraDecalRendererProperties
   ]
  }
  {
   emitterName=SpotLightSecondary
   rendererClasses [&NiagaraMeshRendererProperties, &NiagaraSpriteRendererProperties]
  }
 ]
 userVariables [
  User.bIsFrontEnd {type=NiagaraBool}
  User.bIsFrontEndPreview {type=NiagaraBool}
  User.LightsFade {type=NiagaraFloat, defaultValue=1, description=""}
  User.SongBPM {type=NiagaraFloat, defaultValue=190, description=""}
  User.SpotlightColorArray {
   type=NiagaraDataInterfaceArrayColor
   description=""
   defaultValue {
    struct=NiagaraExt_VariableValue_DataInterface
    value {
     dataInterface=&self:NiagaraDataInterfaceArrayColor_0
     dataInterfaceClass=&NiagaraDataInterfaceArrayColor
    }
   }
  }
  User.SpotlightsPositionArray {
   type=NiagaraDataInterfaceArrayFloat3
   description=""
   defaultValue {
    struct=NiagaraExt_VariableValue_DataInterface
    value {
     dataInterface=&self:NiagaraDataInterfaceArrayFloat3_0
     dataInterfaceClass=&NiagaraDataInterfaceArrayFloat3
    }
   }
  }
 ]
}
