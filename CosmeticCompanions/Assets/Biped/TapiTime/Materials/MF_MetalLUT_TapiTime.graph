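# Material function for the TapiTime biped cosmetic companion that produces a view-dependent metallic tint from a BaseColor input.
# The saturated dot product of camera vector and pixel normal indexes an aluminum reflectance curve in the character curve atlas.
# A selectable channel of that sample is inverted, remapped by SmoothStep, then multiplied by BaseColor and an intensity scalar.
# Curve, channel, range and intensity are parameters in the StrawMetalLUT group.
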MaterialFunction /CosmeticCompanions/Assets/Biped/TapiTime/Materials/MF_MetalLUT_TapiTime
graph {
 properties {libraryCategoriesText []}
 nodes [
  n1 = NamedRerouteDeclaration(None=FunctionInput() {inputName=BaseColor}) {name=BaseColor, nodeColor=<rgba 1 0.1406 0 1>}
  n2 = CameraVectorWS()
  n3 = PixelNormalWS()
  n4 = CurveAtlasRowParameter(CurveTime=Saturate(None=DotProduct(A=@n2, B=@n3))) {
   atlas=/Game/Global/Textures/Default/Curve/FN_Character_Atlas
   curve=/Game/Characters/Player/Female/Medium/Bodies/F_MED_Ms_Alpine/Textures/AluminiumreflectanceCurve
   group=StrawMetalLUT
   parameterName=StrawMetalLUTCurve
  }
  n5 = SmoothStep(
   Min=ScalarParameter() {defaultValue=-0.216, group=StrawMetalLUT, parameterName=StrawLUTMin}
   Max=ScalarParameter() {defaultValue=1, group=StrawMetalLUT, parameterName=SrawLUTMax}
   Value=OneMinus(None=ChannelMaskParameter(None=@n4) {group=StrawMetalLUT, parameterName=StrawLUTChannel}.R)
  )
  FunctionOutput(None=Multiply(A=Multiply(A=@n1, B=@n5), B=ScalarParameter() {defaultValue=1, group=StrawMetalLUT, parameterName=StrawMetalLutIntensity}))
 ]
}
