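# Material function that fakes glass on the TapiTime biped cosmetic companion, overriding only Base Color on the incoming material attributes.
# Warped Fresnel-style ramps index a thin-film texture, whose channels supply a three-color tint, glass mask and specular highlight.
# The highlight also uses virtual-plane UVs facing along the object-to-camera vector.
# A tech-art mask channel and smoothstep limit the effect; the roughness blend is left unconnected to the output.
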MaterialFunction /CosmeticCompanions/Assets/Biped/TapiTime/Materials/MF_FakeGlass_TapiTime
graph {nodes [
  FunctionOutput(None=SetMaterialAttributes!(MaterialAttributes=NamedRerouteUsage(), BaseColor=NamedRerouteUsage()) {attributeSetTypes [#1]})
  n1 = Add(
   A=Multiply(A=NamedRerouteUsage()) {constB=0.2}
   B=Multiply(A=NamedRerouteUsage(), B=ScalarParameter() {defaultValue=0.5, group=FakeGlass, parameterName=FakeGlassIntensity, sortPriority=2})
  )
  n2 = ScalarParameter() {group=FakeGlass, parameterName=ExponentIn}
  Multiply(A=@n1, B=OneMinus(None=Fresnel(ExponentIn=@n2)))
  n3 = GetMaterialAttributes(None=FunctionInput() {inputName=Material Attributes, inputType=FunctionInput_MaterialAttributes, sortPriority=1}) {attributeGetTypes [#2, #1, #3]}
  NamedRerouteDeclaration(None=@n3.MaterialAttributes) {name=Material Attributes, nodeColor=<rgba 0.0781 0 1 1>}
  NamedRerouteDeclaration(None=FunctionInput() {inputName=Fresnel, inputType=FunctionInput_Scalar, sortPriority=2}) {name=In_Fresnel, nodeColor=<rgba 1 0 0.2578 1>}
  n4 = NamedRerouteUsage()
  n5 = NamedRerouteUsage()
  n6 = NamedRerouteUsage()
  n7 = NamedRerouteUsage()
  NamedRerouteDeclaration(None=Add(A=LinearInterpolate(A=@n4, B=@n1, Alpha=@n5), B=Multiply(A=@n6, B=@n7))) {name=Out_BaseColor, nodeColor=<rgba 0.0258 1 0.0122 1>}
  n8 = NamedRerouteUsage()
  n9 = OneMinus(None=NamedRerouteUsage())
  NamedRerouteDeclaration(None=Multiply(A=Multiply(A=@n8, B=@n9), B=ScalarParameter() {defaultValue=0.2, parameterName=FakeSpecStrength})) {name=FakeSpec, nodeColor=<rgba 1 0.8533 0.0329 1>}
  NamedRerouteDeclaration(None=@n3.Specular) {name=In_Specular, nodeColor=<rgba 0.0547 1 0 1>}
  NamedRerouteDeclaration(None=Subtract(A=ObjectPositionWS(), B=CameraPositionWS())) {name=ViewVector, nodeColor=<rgba 0.2656 1 0 1>}
  n10 = NamedRerouteUsage()
  n11 = VectorParameter() {parameterName=VirtualPlaneUVScale, defaultValue=<rgba 40 72.5831 0 0>}
  NamedRerouteDeclaration(None=MaterialFunctionCall(Plane Normal Axis=Normalize(VectorInput=@n10), Plane Center=ObjectPositionWS(), UV Scale=ComponentMask(None=@n11.RGB) {g=true, r=true}) {materialFunction=&VirtualPlaneCoordinates}.UVs) {name=VP_UVs, nodeColor=<rgba 1 0 0.5859 1>}
  n12 = NamedRerouteUsage()
  n13 = TextureSampleParameter2D(UVs=LinearInterpolate(A=NamedRerouteUsage(), B=OneMinus(None=@n12)) {constAlpha=0.3}) {
   group=FakeGlass
   parameterName=FakeGlassTexture
   samplerSource=SSM_Wrap_WorldGroupSettings
   samplerType=SAMPLERTYPE_LinearColor
   sortPriority=1
   texture=/CosmeticCompanions/Assets/Biped/TapiTime/Textures/T_TapiTime_Body_ThinFilmB
  }
  NamedRerouteDeclaration(None=@n13.B) {name=SpecHighlight, nodeColor=<rgba 1 0 0.5156 1>}
  n14 = ScalarParameter() {defaultValue=0.464, group=FakeGlass, parameterName=FakeSpecScaleMask, sortPriority=5}
  n15 = Power(Base=Fresnel() {exponent=1}, Exp=ScalarParameter() {defaultValue=0.452, group=FakeGlass, parameterName=FakeSpecWarpMask, sortPriority=4})
  NamedRerouteDeclaration(None=Add(A=Multiply(A=@n14, B=@n15), B=ScalarParameter() {defaultValue=0.024, group=FakeGlass, parameterName=FakeSpecFilmOffsetMask})) {name=FakeSpecWarping, nodeColor=<rgba 0 1 0.0391 1>}
  NamedRerouteDeclaration(None=@n3.Roughness) {name=In_Roughness, nodeColor=<rgba 0 0.5781 1 1>}
  NamedRerouteDeclaration(None=@n3.BaseColor) {name=In_BaseColor, nodeColor=<rgba 0.8984 1 0 1>}
  NamedRerouteDeclaration(None=VectorParameter() {group=FakeGlass, parameterName=FakeGlassDarkColor}.RGB) {name=GlassDarkColor, nodeColor=<rgba 0.1016 1 0 1>}
  NamedRerouteDeclaration(None=VectorParameter() {group=FakeGlass, parameterName=FakeGlassMidColor, defaultValue=<rgba 0.2708 0.2708 0.2708 0>}.RGB) {name=MidColor, nodeColor=<rgba 1 0 0.4219 1>}
  NamedRerouteDeclaration(None=VectorParameter() {group=FakeGlass, parameterName=FakeGlassHighColor, defaultValue=<rgba 1 1 1 0>}.RGB) {name=HighColor, nodeColor=<rgba 0 0.7422 1 1>}
  n16 = ScalarParameter() {defaultValue=0.432, group=FakeGlass, parameterName=FakeGlassScale, sortPriority=5}
  n17 = Power(Base=NamedRerouteUsage(), Exp=ScalarParameter() {defaultValue=0.212, group=FakeGlass, parameterName=FakeGlassWarp, sortPriority=4})
  n18 = TextureSampleParameter2D(UVs=Add(A=Multiply(A=@n16, B=@n17), B=ScalarParameter() {group=FakeGlass, parameterName=FakeGlassFilmOffset})) {
   group=FakeGlass
   parameterName=FakeGlassTexture
   samplerSource=SSM_Wrap_WorldGroupSettings
   samplerType=SAMPLERTYPE_LinearColor
   sortPriority=1
   texture=&T_TapiTime_Body_ThinFilmB
  }
  NamedRerouteDeclaration(None=MaterialFunctionCall(A=NamedRerouteUsage(), B=NamedRerouteUsage(), C=NamedRerouteUsage(), Alpha=@n18.R) {materialFunction=&Lerp_3Color}.Result) {name=FakeGlassColor, nodeColor=<rgba 0.7578 1 0 1>}
  n19 = ScalarParameter() {defaultValue=0.464, group=FakeGlass, parameterName=FakeGlassScaleMask, sortPriority=5}
  n20 = Power(Base=Fresnel() {exponent=1}, Exp=ScalarParameter() {defaultValue=0.452, group=FakeGlass, parameterName=FakeGlassWarpMask, sortPriority=4})
  n21 = TextureSampleParameter2D(UVs=Add(A=Multiply(A=@n19, B=@n20), B=ScalarParameter() {defaultValue=0.024, group=FakeGlass, parameterName=FakeGlassFilmOffsetMask})) {
   group=FakeGlass
   parameterName=FakeGlassTexture
   samplerSource=SSM_Wrap_WorldGroupSettings
   samplerType=SAMPLERTYPE_LinearColor
   sortPriority=1
   texture=&T_TapiTime_Body_ThinFilmB
  }
  NamedRerouteDeclaration(None=@n21.G) {name=FakeGlassMask, nodeColor=<rgba 0.3594 0 1 1>}
  n22 = ChannelMaskParameter(None=FunctionInput(Preview=TextureSample() {samplerType=SAMPLERTYPE_Masks, texture=/CosmeticCompanions/Assets/Biped/TapiTime/Textures/T_TapiTime_Body_FX}.RGB) {bUsePreviewValueAsDefault=true, inputName=TechArtMask, sortPriority=3}) {group=FakeGlass, parameterName=FakeGlassMaskChannel}
  n23 = Saturate(None=SmoothStep(
    Min=ScalarParameter() {group=FakeGlass, parameterName=Min}
    Max=ScalarParameter() {defaultValue=0.264, group=FakeGlass, parameterName=Max}
    Value=NamedRerouteUsage()
   ))
  NamedRerouteDeclaration(None=Multiply(A=NamedRerouteDeclaration(None=@n22.R) {name=TechArtMaskChannel, nodeColor=<rgba 0.5 1 0 1>}, B=OneMinus(None=@n23))) {name=CombinedMask, nodeColor=<rgba 1 0 0.1875 1>}
  n24 = LinearInterpolate(Alpha=NamedRerouteUsage()) {constA=1, constB=0.01}
  n25 = Power(Base=NamedRerouteUsage(), Exp=ScalarParameter() {defaultValue=2, group=FakeGlass, parameterName=FakeGlassExponent, sortPriority=3})
  Saturate(None=LinearInterpolate(A=NamedRerouteUsage(), B=Multiply(A=@n24, B=@n25), Alpha=ScalarParameter() {group=FakeGlass, parameterName=FakeGlassRoughnessInfluence}))
 ]}
