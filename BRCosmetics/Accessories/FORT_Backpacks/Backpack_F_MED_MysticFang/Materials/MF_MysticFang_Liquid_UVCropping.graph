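# Material function for the Mystic Fang back bling's liquid material that optionally crops UVs to a parameterized region for a boba mask.
# The Cropping Region vector feeds a shared texture cropping function, with R and G as the upper-left corner and B as both lower-right coordinates.
# It outputs cropped UVs plus a hard mask from the ceiling of saturated cropped V.
# A static bool switches to passthrough UVs and a mask of one.
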MaterialFunction /BRCosmetics/Accessories/FORT_Backpacks/Backpack_F_MED_MysticFang/Materials/MF_MysticFang_Liquid_UVCropping
graph {nodes [
  n1 = FunctionInput(Preview=TextureCoordinate()) {bUsePreviewValueAsDefault=true, inputName=UVs, inputType=FunctionInput_Vector2}
  n2 = VectorParameter() {group=Boba, parameterName=Cropping Region, defaultValue=<rgba 0.006 0.866 0.56 1>}
  n3 = MaterialFunctionCall!(
   UVs=@n1
   TextureIn
   UpperLeftCorner=ComponentMask(None=@n2.RGB) {g=true, r=true}
   LowerRightCorner=AppendVector(A=@n2.B, B=@n2.B)
   ExportFloat4
  ) {materialFunction=/BRCosmetics/Characters/Player/Female/Medium/Bodies/F_MED_SpatialTravel_Watt/Skins/Superlevels/ParentMaterials/MF_TextureCropping_SL}
  n4 = StaticBoolParameter() {group=Boba, parameterName=Use Cropped Boba Mask}
  FunctionOutput(None=StaticSwitch(True=@n3.Crop UVs, False=@n1, Value=@n4)) {outputName=Cropped UVs, sortPriority=-1}
  n5 = ComponentMask(None=Saturate(None=@n3.Crop UVs)) {g=true}
  FunctionOutput(None=StaticSwitch(True=Ceil(None=@n5), False=Constant() {r=1}, Value=@n4)) {outputName=Crop Mask}
 ]}
