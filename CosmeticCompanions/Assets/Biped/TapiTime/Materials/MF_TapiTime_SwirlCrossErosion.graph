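# Material function for the TapiTime companion's liquid swirl, outputting a looping grayscale mask from two crossfaded, eroding noise layers.
# Duration drives two half-cycle-offset phases, whose triangle waves, raised to Swirl Power, set each layer's erosion threshold.
# Each layer scales In_UV about center, interpolating Min to Max Scale logarithmically, pans over time and adds Depth-driven BumpOffset parallax.
# Local-position UV projection nodes and the TechArt mask input do not reach the output.
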MaterialFunction /CosmeticCompanions/Assets/Biped/TapiTime/Materials/MF_TapiTime_SwirlCrossErosion
graph {nodes [
  n1 = Multiply(A=LocalPosition() {localOrigin=InstancePreSkinning}.XYZ) {constB=0.075}
  n2 = VertexInterpolator(VS=ComponentMask(None=Add(A=@n1) {constB=-0.3314}) {b=true})
  n3 = OneMinus(None=Saturate(None=Abs(None=@n2.PS)))
  n4 = Add(A=Multiply(A=LocalPosition() {localOrigin=InstancePreSkinning}.XYZ) {constB=0.075}) {constB=-0.3314}
  AppendVector(A=@n3, B=VertexInterpolator(VS=ComponentMask(None=@n4) {g=true}).PS)
  ComponentMask(None=Multiply(A=LocalPosition() {localOrigin=InstancePreSkinning}.XYZ) {constB=0.003}) {g=true, r=true}
  n5 = FunctionInput() {inputName=In_UV, inputType=FunctionInput_Vector2, sortPriority=5}
  n6 = Constant() {r=10}
  n7 = Logarithm10!(X=FunctionInput(Preview=Constant() {r=0.5}) {inputName=Min Scale, inputType=FunctionInput_Scalar})
  n8 = Logarithm10(X=FunctionInput(Preview=Constant() {r=2}) {inputName=Max Scale, inputType=FunctionInput_Scalar})
  n9 = Power(Base=@n6, Exp=LinearInterpolate(A=@n7, B=@n8, Alpha=NamedRerouteUsage()))
  n10 = Multiply(A=Time(), B=Constant2Vector() {g=0.03})
  n11 = Multiply(A=FunctionInput() {inputName=Depth, inputType=FunctionInput_Scalar}) {constB=-1}
  n12 = NamedRerouteUsage()
  n13 = BumpOffset(
   Coordinate=Add(
    A=MaterialFunctionCall(UVs=@n5, Texture Scale=@n9) {materialFunction=&ScaleUVsByCenter}.UVs
    B=@n10
   )
   Height=Multiply(A=@n11, B=OneMinus(None=@n12)) {constB=-5}
  )
  n14 = TextureObjectParameter() {
   group=Liquid - Swirl
   parameterName=Swirl Texture
   samplerType=SAMPLERTYPE_LinearGrayscale
   sortPriority=0
   texture=/BRCosmetics/Accessories/FORT_Backpacks/Backpack_F_MED_MysticFang/Textures/T_SwirlNoise
  }
  n15 = Constant() {r=5}
  n16 = Divide(
   A=Time()
   B=NamedRerouteDeclaration(None=FunctionInput(Preview=@n15) {inputName=Duration, inputType=FunctionInput_Scalar}) {name=FadeDuration, nodeColor=<rgba 1 0 0.5156 1>}
  )
  n17 = ConstantBiasScale(None=NamedRerouteDeclaration(None=Frac(None=@n16)) {name=Layer1_alpha, nodeColor=<rgba 0.3828 1 0 1>}) {bias=-0.5, scale=2}
  n18 = ScalarParameter() {defaultValue=2, group=Liquid - Swirl, parameterName=Swirl Power, sortPriority=1}
  n19 = Power(Base=NamedRerouteDeclaration(None=Abs(None=@n17)) {name=CrossFade_Alpha, nodeColor=<rgba 0 0.8359 1 1>}, Exp=@n18)
  n20 = Divide(
   A=Subtract(
    A=TextureSample(UVs=@n13, Tex=@n14) {samplerType=SAMPLERTYPE_LinearGrayscale, texture=/Game/Global/Textures/Default/Blanks/T_BlackLinearGrayscale}.R
    B=@n19
   )
   B=Subtract(B=@n19)
  )
  n21 = Power(Base=@n6, Exp=LinearInterpolate(A=@n7, B=@n8, Alpha=NamedRerouteUsage()))
  n22 = NamedRerouteUsage()
  n23 = BumpOffset(
   Coordinate=Add(
    A=MaterialFunctionCall(UVs=@n5, Texture Scale=@n21) {materialFunction=&ScaleUVsByCenter}.UVs
    B=@n10
   )
   Height=Multiply(A=OneMinus(None=@n22), B=@n11) {constB=-5}
  )
  n24 = NamedRerouteDeclaration(None=Frac(None=Add(A=@n16) {constB=0.5})) {name=Layer2_alpha, nodeColor=<rgba 1 0 0.7031 1>}
  n25 = Power(Base=Abs(None=ConstantBiasScale(None=@n24) {bias=-0.5, scale=2}), Exp=@n18)
  n26 = Divide(
   A=Subtract(
    A=TextureSample(UVs=@n23, Tex=@n14) {samplerType=SAMPLERTYPE_LinearGrayscale, texture=&T_BlackLinearGrayscale}.R
    B=@n25
   )
   B=Subtract(B=@n25)
  )
  FunctionOutput(None=LinearInterpolate(A=Saturate(None=@n20), B=Saturate(None=@n26), Alpha=NamedRerouteUsage()))
  n27 = TextureCoordinate()
  n28 = Constant2Vector() {g=2.1515, r=1.3875}
  n29 = Multiply(A=@n27, B=@n28)
  n30 = Constant2Vector() {g=0.5215, r=0.0078}
  n31 = FunctionInput(Preview=Constant2Vector() {g=0.5, r=0.5}) {inputName=Scale Pivot, inputType=FunctionInput_Vector2}
  n32 = Multiply(A=LocalPosition() {localOrigin=InstancePreSkinning}.XYZ, B=Constant3Vector() {constant=<rgba 0.0716 0.03 0.03 0>}) {constB=0.03}
  LinearInterpolate(
   A=Subtract(A=Add(A=@n29, B=@n30), B=Subtract(A=@n31) {constB=0.5})
   B=VertexInterpolator(VS=ComponentMask(None=@n32) {b=true, g=true}).PS
   Alpha=@n3
  )
  MaterialFunctionCall!(Texture Scale=@n28, UVs=@n27, Offset XY Texturesize Z) {materialFunction=/Engine/Functions/Engine_MaterialFunctions01/Texturing/ScaleUVsAroundPoint}
  NamedRerouteDeclaration(None=FunctionInput() {inputName=In_TechArtMask, sortPriority=3}) {name=In_TechArtMask, nodeColor=<rgba 0 1 0.6484 1>}
 ]}
