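# Material function that adds boba-tea liquid to the TapiTime biped cosmetic companion, overriding base color and emissive on incoming material attributes using a tech-art mask.
# Pre-skinned local position drives liquid gradients, with upper/lower colors, eroded swirls and Fresnel glow.
# Three boba layers sample channels of a packed mask through cropped BumpOffset UVs for parallax depth, moved by drift, sine wobble and jump-offset parameters.
# Smoothstepped layers max-combine into the boba mask.
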MaterialFunction /CosmeticCompanions/Assets/Biped/TapiTime/Materials/MF_TapiTime_Liquid
graph {nodes [
  n1 = ScalarParameter() {defaultValue=0.7, group=Boba, parameterName=Layer2_BobaHardness}
  Add(A=@n1, B=NamedRerouteUsage())
  n2 = Sine(None=Add(A=NamedRerouteUsage(), B=ScalarParameter() {defaultValue=0.7, group=Boba, parameterName=Layer2_FadeOffsetTime}))
  ConstantBiasScale(None=@n2)
  n3 = NamedRerouteUsage()
  n4 = Multiply(A=Multiply(A=NamedRerouteUsage(), B=NamedRerouteUsage()), B=NamedRerouteUsage())
  n5 = LinearInterpolate(
   A=NamedRerouteUsage()
   B=NamedRerouteUsage()
   Alpha=Multiply(A=@n4, B=TextureSample() {texture=/CosmeticCompanions/Assets/Biped/TapiTime/Textures/T_TapiTime_Body_UVFix}.R)
  )
  n6 = NamedRerouteUsage()
  n7 = NamedRerouteUsage()
  n8 = LinearInterpolate(
   A=VectorParameter() {group=Liquid, parameterName=Glow Color, sortPriority=10, defaultValue=<rgba 0.5 0.0417 0 1>}.RGB
   Alpha=Fresnel(ExponentIn=ScalarParameter() {defaultValue=0.2, group=Liquid, parameterName=Glow Fresnel Exp, sortPriority=11})
  ) {constB=0}
  n9 = LinearInterpolate(A=@n8, B=Multiply(A=@n8, B=NamedRerouteUsage()) {constB=0.1}, Alpha=NamedRerouteUsage()) {constB=0}
  n10 = Multiply(A=NamedRerouteUsage(), B=TextureSample() {texture=&T_TapiTime_Body_UVFix}.R)
  FunctionOutput(None=SetMaterialAttributes!(
    MaterialAttributes=NamedRerouteUsage()
    BaseColor=LinearInterpolate(A=@n3, B=@n5, Alpha=@n6)
    EmissiveColor=LinearInterpolate(A=@n7, B=@n9, Alpha=@n10)
   ) {attributeSetTypes [#1, #2]})
  n11 = Multiply(A=LocalPosition() {localOrigin=InstancePreSkinning}.XYZ) {constB=0.03}
  n12 = VectorParameter() {group=Boba, parameterName=BobaUVSpeedAndOffsets, defaultValue=<rgba 1 1 0 0>}
  n13 = Add(A=Multiply(A=TextureCoordinate(), B=ComponentMask(None=@n12.RGB) {g=true, r=true}), B=ComponentMask(None=@n12.RGBA) {a=true, b=true})
  LinearInterpolate(A=VertexInterpolator(VS=ComponentMask(None=@n11) {g=true, r=true}).PS, B=@n13, Alpha=ComponentMask(None=NamedRerouteUsage()) {r=true})
  n14 = Multiply(A=LocalPosition() {localOrigin=InstancePreSkinning}.XYZ, B=Constant3Vector() {constant=<rgba 0.0716 0.03 0.03 0>}) {constB=0.03}
  LinearInterpolate(A=VertexInterpolator(VS=ComponentMask(None=@n14) {b=true, g=true}).PS, B=TextureCoordinate(), Alpha=NamedRerouteUsage())
  n15 = TextureSample(UVs=NamedRerouteUsage(), Tex=NamedRerouteUsage()) {samplerSource=SSM_Wrap_WorldGroupSettings, samplerType=SAMPLERTYPE_LinearGrayscale}
  MaterialExpressionMultiply_3 = Multiply(A=NamedRerouteUsage(), B=@n15.RGB)
  n16 = Convert(None=@MaterialExpressionMultiply_3) {
   nodeName=Break Float3
   convertInputs [{
     type=Vector3
     expressionInput {
      expression=&self:MaterialExpressionMultiply_3
      inputName=None
      mask=0
      maskA=0
      maskB=0
      maskG=0
      maskR=0
      outputIndex=0
     }
    }]
   convertOutputs [{type=Scalar}, {type=Scalar}]
  }
  n17 = Constant() {r=0.25}
  n18 = Ceil(None=Subtract(A=@n16, B=@n17) {constB=0.1})
  n19 = Constant() {r=0.5}
  n20 = Ceil(None=Subtract(A=@n16, B=@n19) {constB=0.3})
  n21 = OneMinus(None=Ceil(None=Subtract(A=@n16, B=@n17) {constB=0.1}))
  Max(A=Multiply(A=Subtract(A=@n18, B=@n20), B=@n16), B=Multiply(A=Saturate(None=@n21), B=@n16))
  n22 = TextureSample(UVs=NamedRerouteUsage(), Tex=NamedRerouteUsage()) {samplerSource=SSM_Wrap_WorldGroupSettings, samplerType=SAMPLERTYPE_LinearGrayscale}
  MaterialExpressionMultiply_0 = Multiply(A=NamedRerouteUsage(), B=@n22.RGB)
  n23 = Convert(None=@MaterialExpressionMultiply_0) {
   nodeName=Break Float3
   convertInputs [{
     type=Vector3
     expressionInput {
      expression=&self:MaterialExpressionMultiply_0
      inputName=None
      mask=0
      maskA=0
      maskB=0
      maskG=0
      maskR=0
      outputIndex=0
     }
    }]
   convertOutputs [{type=Scalar}, {type=Scalar}]
  }
  n24 = Constant() {r=0.75}
  n25 = Subtract(A=@n23, B=@n24) {constB=0.6}
  n26 = Subtract(A=@n23, B=Constant() {r=1})
  Multiply(A=Subtract(A=Ceil(None=@n25), B=Ceil(None=@n26)), B=@n23)
  n27 = TextureSample(UVs=NamedRerouteUsage(), Tex=NamedRerouteUsage()) {samplerSource=SSM_Wrap_WorldGroupSettings, samplerType=SAMPLERTYPE_LinearGrayscale}
  MaterialExpressionMultiply_1 = Multiply(A=NamedRerouteUsage(), B=@n27.RGB)
  n28 = Convert(None=@MaterialExpressionMultiply_1) {
   nodeName=Break Float3
   convertInputs [{
     type=Vector3
     expressionInput {
      expression=&self:MaterialExpressionMultiply_1
      inputName=None
      mask=0
      maskA=0
      maskB=0
      maskG=0
      maskR=0
      outputIndex=0
     }
    }]
   convertOutputs [{type=Scalar}, {type=Scalar}]
  }
  n29 = Subtract(A=@n28, B=@n19) {constB=0.3}
  n30 = Subtract(A=@n28, B=@n24) {constB=0.6}
  Multiply(A=Subtract(A=Ceil(None=@n29), B=Ceil(None=@n30)), B=@n28)
  n31 = OneMinus(None=ScalarParameter() {defaultValue=0.7, group=Boba, parameterName=Layer1_BobaHardness})
  Multiply(A=@n31, B=NamedRerouteUsage())
  NamedRerouteDeclaration(None=FunctionInput() {inputName=Mat, inputType=FunctionInput_MaterialAttributes}) {name=In_MaterialAttributes, nodeColor=<rgba 1 0 0.1641 1>}
  NamedRerouteDeclaration(None=FunctionInput() {inputName=TechArtMask, sortPriority=10}) {name=In_TechArtMask, nodeColor=<rgba 0 0.4844 1 1>}
  NamedRerouteDeclaration(None=LinearInterpolate(Alpha=@n4) {constA=0.5, constB=0}) {name=Fake Subsurface Mask, nodeColor=<rgba 0.9453 1 0 1>}
  n32 = TextureObjectParameter() {
   group=Boba
   parameterName=Boba Mask
   samplerType=SAMPLERTYPE_LinearColor
   sortPriority=0
   texture=/CosmeticCompanions/Assets/Biped/TapiTime/Textures/T_TapiTime_Body_BobaPacked
  }
  NamedRerouteDeclaration(None=@n32) {name=Boba Mask, nodeColor=<rgba 1 0.2344 0 1>}
  n33 = VertexInterpolator(VS=LocalPosition() {localOrigin=InstancePreSkinning}.XYZ)
  n34 = ComponentMask(None=@n33.PS) {b=true}
  n35 = Multiply(A=ComponentMask(None=@n33.PS) {r=true}, B=ScalarParameter() {group=Liquid - Swirl, parameterName=LiquidUVXOffset})
  n36 = ComponentMask(None=@n33.PS) {g=true}
  n37 = ScalarParameter() {defaultValue=0.16, group=Liquid - Swirl, parameterName=LiquidUVYOffset}
  n38 = Subtract(A=Add(A=Add(A=@n34, B=@n35), B=Multiply(A=@n36, B=@n37)), B=Constant() {r=-9})
  n39 = ScalarParameter() {defaultValue=55, parameterName=LiquidUVCoordScale}
  NamedRerouteDeclaration(None=Saturate(None=Divide(A=@n38, B=@n39))) {name=DensityGradient, nodeColor=<rgba 0.0078 0 1 1>}
  n40 = ScalarParameter() {defaultValue=8, group=Liquid - Swirl, parameterName=Swirl Power}
  n41 = Multiply(A=NamedRerouteUsage()) {constB=6}
  n42 = Power(Base=NamedRerouteUsage(), Exp=ScalarParameter() {defaultValue=1, group=Liquid, parameterName=Liquid Transition Power, sortPriority=1})
  n43 = NamedRerouteUsage()
  n44 = Multiply(A=Saturate(None=Subtract(A=@n42, B=@n43) {constB=0.25}), B=Add(B=@n43) {constA=1}) {constB=1.25}
  NamedRerouteDeclaration(None=Multiply(A=Subtract(A=@n40, B=@n41), B=Saturate(None=@n44))) {name=SwirlDensityExponent, nodeColor=<rgba 0.3125 1 0 1>}
  NamedRerouteDeclaration(None=Multiply(
    A=ScalarParameter() {group=Boba, parameterName=Jump_Offset}
    B=ScalarParameter() {defaultValue=0.5, group=Boba, parameterName=Jump Multiplier}
   )) {name=Jump_Offset, nodeColor=<rgba 0 1 0.7188 1>}
  n45 = ScalarParameter() {defaultValue=0.016, group=Boba, parameterName=Layer1_XDriftDistance}
  n46 = NamedRerouteUsage()
  n47 = Add(A=NamedRerouteUsage(), B=Multiply(A=NamedRerouteUsage(), B=Multiply(A=@n45, B=@n46)) {constB=0.015})
  n48 = ConstantBiasScale(None=NamedRerouteUsage()) {bias=0.15, scale=0.25}
  n49 = NamedRerouteUsage()
  n50 = NamedRerouteUsage()
  n51 = Sine(None=Add(A=Add(A=@n49, B=@n50), B=ScalarParameter() {defaultValue=0.4, group=Boba, parameterName=Layer1_OffsetTime}))
  n52 = Multiply(
   A=NamedRerouteDeclaration(None=ConstantBiasScale(None=@n51)) {name=Layer1Wobble, nodeColor=<rgba 0.8047 1 0 1>}
   B=Multiply(A=ScalarParameter() {defaultValue=0.075, group=Boba, parameterName=Layer1_YDriftDistance}, B=NamedRerouteUsage())
  ) {constB=0.35}
  n53 = Add(A=NamedRerouteUsage(), B=Add(A=Multiply(A=@n48, B=@n52), B=NamedRerouteUsage()))
  n54 = MaterialFunctionCall!(UVs=BumpOffset(
    Coordinate=MaterialFunctionCall(X=@n47, Y=@n53) {materialFunction=&MakeFloat2}.Result
    Height=ScalarParameter() {defaultValue=0.7, group=Boba, parameterName=Layer1_OffsetDepth}
    HeightRatioInput=NamedRerouteUsage()
   )) {materialFunction=/BRCosmetics/Accessories/FORT_Backpacks/Backpack_F_MED_MysticFang/Materials/MF_MysticFang_Liquid_UVCropping}
  NamedRerouteDeclaration(None=@n54.Cropped UVs) {name=Layer1CroppedUVs, nodeColor=<rgba 1 0 0.6328 1>}
  NamedRerouteDeclaration(None=@n54.Crop Mask) {name=Layer1CroppedUVsMask, nodeColor=<rgba 0 0.9531 1 1>}
  NamedRerouteDeclaration(None=ScalarParameter() {defaultValue=0.075, group=Boba, parameterName=Offset_Height_Ratio}) {name=Masked_HeightRatio, nodeColor=<rgba 0.0312 1 0 1>}
  n55 = NamedRerouteUsage()
  n56 = Multiply(A=ScalarParameter() {defaultValue=-0.006, group=Boba, parameterName=Layer2_DriftDistance}, B=NamedRerouteUsage())
  n57 = Add(A=Add(A=Multiply(A=@n55, B=@n56) {constB=0.015}, B=NamedRerouteUsage()), B=ScalarParameter() {group=Boba, parameterName=Layer2_XOffset})
  n58 = ConstantBiasScale(None=NamedRerouteUsage()) {bias=0.15, scale=0.25}
  n59 = NamedRerouteUsage()
  n60 = NamedRerouteUsage()
  n61 = Sine(None=Add(A=Add(A=@n59, B=@n60), B=ScalarParameter() {defaultValue=0.7, group=Boba, parameterName=Layer2_OffsetTime}))
  n62 = Multiply(
   A=NamedRerouteDeclaration(None=ConstantBiasScale(None=@n61)) {name=Layer2Wobble, nodeColor=<rgba 0 1 0.1562 1>}
   B=Multiply(A=ScalarParameter() {defaultValue=0.05, group=Boba, parameterName=Layer2_YDriftDistance}, B=NamedRerouteUsage())
  ) {constB=0.35}
  n63 = Add(A=NamedRerouteUsage(), B=Add(A=Multiply(A=@n58, B=@n62), B=NamedRerouteUsage()))
  n64 = MaterialFunctionCall(UVs=BumpOffset(
    Coordinate=MaterialFunctionCall(X=@n57, Y=@n63) {materialFunction=&MakeFloat2}.Result
    Height=ScalarParameter() {defaultValue=0.7, group=Boba, parameterName=Layer2_OffsetDepth}
    HeightRatioInput=NamedRerouteUsage()
   )) {materialFunction=&MF_MysticFang_Liquid_UVCropping}
  NamedRerouteDeclaration(None=@n64.Cropped UVs) {name=Layer2CroppedUVs, nodeColor=<rgba 1 0.7266 0 1>}
  NamedRerouteDeclaration(None=Multiply(A=Time(), B=ScalarParameter() {defaultValue=0.3, group=Boba, parameterName=BobaWobbleSpeed})) {name=WobbleTime, nodeColor=<rgba 1 0.375 0 1>}
  NamedRerouteDeclaration(None=@n64.Crop Mask) {name=Layer2CroppedMask, nodeColor=<rgba 0.4062 0 1 1>}
  n65 = ScalarParameter() {defaultValue=0.008, group=Boba, parameterName=Layer3_DriftDistance}
  n66 = NamedRerouteUsage()
  n67 = Add(A=Multiply(A=NamedRerouteUsage(), B=Multiply(A=@n65, B=@n66)) {constB=0.015}, B=NamedRerouteUsage())
  n68 = ConstantBiasScale(None=NamedRerouteUsage()) {bias=0.15, scale=0.25}
  n69 = NamedRerouteUsage()
  n70 = NamedRerouteUsage()
  n71 = Sine(None=Add(A=Add(A=@n69, B=@n70), B=ScalarParameter() {defaultValue=0.2, group=Boba, parameterName=Layer3_OffsetTime}))
  n72 = Multiply(
   A=NamedRerouteDeclaration(None=ConstantBiasScale(None=@n71)) {name=Layer3Wobble, nodeColor=<rgba 0.875 0 1 1>}
   B=Multiply(A=ScalarParameter() {defaultValue=0.0665, group=Boba, parameterName=Layer3_YDriftDistance}, B=NamedRerouteUsage())
  ) {constB=0.35}
  n73 = Add(A=NamedRerouteUsage(), B=Add(A=Multiply(A=@n68, B=@n72), B=NamedRerouteUsage()))
  n74 = MaterialFunctionCall(UVs=BumpOffset(
    Coordinate=MaterialFunctionCall(X=@n67, Y=@n73) {materialFunction=&MakeFloat2}.Result
    Height=ScalarParameter() {defaultValue=0.7, group=Boba, parameterName=Layer3_OffsetDepth}
    HeightRatioInput=NamedRerouteUsage()
   )) {materialFunction=&MF_MysticFang_Liquid_UVCropping}
  NamedRerouteDeclaration(None=@n74.Cropped UVs) {name=Layer3CroppedUVs, nodeColor=<rgba 0 1 0.0859 1>}
  NamedRerouteDeclaration(None=@n74.Crop Mask) {name=Layer3CropMask, nodeColor=<rgba 1 0 0.2344 1>}
  n75 = NamedRerouteUsage()
  n76 = ScalarParameter() {defaultValue=0.052, group=Boba, parameterName=Layer2_JumpOffset}
  NamedRerouteDeclaration(None=Add(A=Multiply(A=@n75, B=@n76) {constB=0.75}, B=ScalarParameter() {defaultValue=-0.005, group=Boba, parameterName=Layer2_BaseOffset})) {name=Layer2JumpOffset, nodeColor=<rgba 0 0.5547 1 1>}
  n77 = NamedRerouteUsage()
  n78 = ScalarParameter() {defaultValue=0.052, group=Boba, parameterName=Layer3_JumpOffset}
  NamedRerouteDeclaration(None=Add(A=Multiply(A=@n77, B=@n78) {constB=0.75}, B=ScalarParameter() {defaultValue=-0.0075, group=Boba, parameterName=Layer3_BaseOffset})) {name=Layer3JumpOffset, nodeColor=<rgba 0.875 1 0 1>}
  n79 = NamedRerouteUsage()
  n80 = ScalarParameter() {defaultValue=0.056, group=Boba, parameterName=Layer1_JumpOffset}
  NamedRerouteDeclaration(None=Add(A=Multiply(A=@n79, B=@n80) {constB=0.75}, B=ScalarParameter() {defaultValue=-0.005, group=Boba, parameterName=Layer1_BaseOffset})) {name=Layer1JumpOffset, nodeColor=<rgba 0.8047 0 1 1>}
  NamedRerouteDeclaration(None=ComponentMask(None=@n13) {r=true}) {name=UVx, nodeColor=<rgba 1 0 0 1>}
  NamedRerouteDeclaration(None=ComponentMask(None=@n13) {g=true}) {name=UVy, nodeColor=<rgba 0 1 0 1>}
  n81 = OneMinus(None=ScalarParameter() {defaultValue=0.9, group=Boba, parameterName=Layer1_BobaSize})
  n82 = SmoothStep(Min=@n81, Max=Add(A=@n81, B=@n31) {constB=0.5}, Value=Multiply(A=NamedRerouteUsage(), B=@n22.R))
  n83 = OneMinus(None=ScalarParameter() {defaultValue=0.9, group=Boba, parameterName=Layer2_BobaSize})
  n84 = OneMinus(None=@n1)
  n85 = NamedRerouteUsage()
  n86 = OneMinus(None=ScalarParameter() {defaultValue=0.9, group=Boba, parameterName=Layer3_BobaSize})
  n87 = OneMinus(None=ScalarParameter() {defaultValue=0.7, group=Boba, parameterName=Layer3_BobaHardness})
  n88 = NamedRerouteUsage()
  n89 = Max(
   A=SmoothStep(Min=@n83, Max=Add(A=@n83, B=@n84) {constB=0.5}, Value=Multiply(A=@n85, B=@n27.G))
   B=SmoothStep(Min=@n86, Max=Add(A=@n86, B=@n87) {constB=0.5}, Value=Multiply(A=@n88, B=@n15.B))
  )
  NamedRerouteDeclaration(None=NamedRerouteDeclaration(None=Max(A=@n82, B=@n89)) {name=Boba_Color_Mask, nodeColor=<rgba 0.3594 0 1 1>}) {name=Boba_Mask, nodeColor=<rgba 1 0.9141 0 1>}
  n90 = MaterialFunctionCall(Color=NamedRerouteUsage()) {materialFunction=&SplitComponents}
  NamedRerouteDeclaration(None=@n90.B) {name=BobaMask, nodeColor=<rgba 0.5 0 1 1>}
  n91 = GetMaterialAttributes(None=NamedRerouteUsage()) {attributeGetTypes [#1, #2]}
  NamedRerouteDeclaration(None=@n91.BaseColor) {name=In_BaseColor, nodeColor=<rgba 0 1 0.1797 1>}
  NamedRerouteDeclaration(None=@n91.EmissiveColor) {name=In_EmissiveColor, nodeColor=<rgba 1 0 0.1406 1>}
  n92 = LocalPosition() {localOrigin=InstancePreSkinning}
  NamedRerouteDeclaration(None=Multiply(A=VertexInterpolator(VS=@n92.XY).PS) {constB=0.03}) {name=VP_UVs, nodeColor=<rgba 1 0 0.5859 1>}
  NamedRerouteDeclaration(None=Multiply(A=@n2, B=ScalarParameter() {defaultValue=0.05, group=Boba, parameterName=Layer2_FadeMultiplier}) {constB=0.35}) {name=Layer2FadeInOutOffset, nodeColor=<rgba 0 1 0.1562 1>}
  NamedRerouteDeclaration(None=Multiply(A=Time(), B=ScalarParameter() {defaultValue=0.05, group=Boba, parameterName=BobaFadeInOutSpeed})) {name=BobaFadeTime, nodeColor=<rgba 1 0.375 0 1>}
  NamedRerouteDeclaration(None=@n90.G) {name=LiquidSwirlsMask, nodeColor=<rgba 1 0.9335 0 1>}
  n93 = Add(A=Multiply(A=LocalPosition() {localOrigin=InstancePreSkinning}.XYZ) {constB=0.02}) {constB=-0.0831}
  n94 = SmoothStep(Value=Saturate(None=Abs(None=@n93))) {constMax=0.5, constMin=0.0824}
  NamedRerouteDeclaration(None=VertexInterpolator(VS=ComponentMask(None=@n94) {g=true}).PS) {name=SeamHideGradient, nodeColor=<rgba 0 1 0.625 1>}
  n95 = VectorParameter() {group=Liquid, parameterName=UpperLiquidColor, sortPriority=-1, defaultValue=<rgba 0.05 0.02 0.2 1>}
  n96 = VectorParameter() {group=Liquid, parameterName=LowerLiquidColor, sortPriority=-1, defaultValue=<rgba 0.8 0.48 0.7828 1>}
  NamedRerouteDeclaration(None=LinearInterpolate(A=@n95.RGB, B=@n96.RGB, Alpha=NamedRerouteUsage())) {name=LiquidColor, nodeColor=<rgba 1 0.5391 0 1>}
  n97 = ComponentMask(None=Multiply(A=LocalPosition() {localOrigin=InstancePreSkinning}.XYZ, B=Constant3Vector() {constant=<rgba 0.0716 0.03 0.03 0>}) {constB=0.03}) {b=true, g=true}
  NamedRerouteDeclaration(None=LinearInterpolate(A=VertexInterpolator(VS=@n97).PS, B=TextureCoordinate(), Alpha=NamedRerouteUsage())) {name=UVLerp, nodeColor=<rgba 1 0.3047 0 1>}
  NamedRerouteDeclaration(None=LinearInterpolate(A=Constant() {r=3}, B=Constant() {r=25}, Alpha=NamedRerouteUsage())) {name=DepthLerp, nodeColor=<rgba 0 0.0156 1 1>}
  n98 = NamedRerouteUsage()
  NamedRerouteDeclaration(None=Add(A=Constant() {r=0.35}, B=Multiply(A=@n98) {constB=0.15})) {name=offset, nodeColor=<rgba 0 1 0.25 1>}
  n99 = VectorParameter() {group=Liquid - Swirl, parameterName=SwirlCrossErosionPivot}
  n100 = Add(
   A=MaterialFunctionCall!(
    Min Scale=ScalarParameter() {defaultValue=0.5, group=Liquid - Swirl, parameterName=SwirlCrossErosionMinScale}
    Max Scale=ScalarParameter() {defaultValue=2, group=Liquid - Swirl, parameterName=SwirlCrossErosionMaxScale}
    Scale Pivot=ComponentMask(None=@n99.RGB) {g=true, r=true}
    Depth=NamedRerouteUsage()
    Duration=ScalarParameter() {defaultValue=10, group=Liquid - Swirl, parameterName=Swirl Cross Duration}
    In_TechArtMask=NamedRerouteUsage()
    In_UV=NamedRerouteUsage()
   ) {materialFunction=&/MF_TapiTime_SwirlCrossErosion}.Result
   B=NamedRerouteUsage()
  )
  n101 = NamedRerouteUsage()
  NamedRerouteDeclaration(None=Saturate(None=Power(Base=@n100, Exp=@n101))) {name=LiquidMask, nodeColor=<rgba 0 0.0156 1 1>}
  NamedRerouteDeclaration(None=LinearInterpolate(
    A=VectorParameter() {group=Boba, parameterName=Boba Ring Color, sortPriority=1, defaultValue=<rgba 0 0 0 1>}.RGB
    B=VectorParameter() {group=Boba, parameterName=Boba Core Color, sortPriority=1, defaultValue=<rgba 0 0 0 1>}.RGB
    Alpha=NamedRerouteUsage()
   )) {name=Boba Color, nodeColor=<rgba 0.3359 1 0 1>}
  NamedRerouteDeclaration(None=ScalarParameter() {defaultValue=1, group=Boba, parameterName=Drift Multiplier}) {name=DriftMultiplier, nodeColor=<rgba 1 0.9141 0 1>}
  n102 = NamedRerouteUsage()
  NamedRerouteDeclaration(None=SmoothStep(Min=Add(A=@n102) {constB=0.4}, Max=Subtract(A=@n102) {constB=0.3}, Value=NamedRerouteUsage())) {name=DenseFloor, nodeColor=<rgba 0 0.625 1 1>}
 ]}
