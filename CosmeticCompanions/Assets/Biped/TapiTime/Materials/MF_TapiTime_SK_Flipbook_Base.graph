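# Material function that adds a flipbook face of brows, eyes and mouth to the TapiTime biped cosmetic companion's material.
# Per-feature curve functions supply frame indices, which are remapped by vector parameters before flipbook sampling; FaceColorize blends eye and mouth variants.
# Results can override base color, emissive, opacity mask and vertex-color-driven WPO, with optional extra hide mask and debug views.
# Extra outputs expose eye and mouth alphas and scaled-UV masks.
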MaterialFunction /CosmeticCompanions/Assets/Biped/TapiTime/Materials/MF_TapiTime_SK_Flipbook_Base
graph {nodes [
  n1 = NamedRerouteUsage()
  n2 = NamedRerouteUsage()
  n3 = NamedRerouteUsage()
  n4 = SetMaterialAttributes!(
   MaterialAttributes=NamedRerouteUsage()
   BaseColor=StaticSwitch(True=NamedRerouteUsage(), False=NamedRerouteUsage(), Value=NamedRerouteUsage())
   EmissiveColor=NamedRerouteUsage()
   OpacityMask=NamedRerouteUsage()
   WorldPositionOffset=StaticSwitch(True=NamedRerouteUsage(), False=StaticSwitch(True=@n1, False=@n2, Value=@n3), Value=NamedRerouteUsage())
  ) {attributeSetTypes [
    #1
    #2
    #3
    #4
   ]}
  FunctionOutput(None=@n4)
  FunctionOutput(None=NamedRerouteUsage()) {outputName=EyesAlpha, sortPriority=1}
  FunctionOutput(None=NamedRerouteUsage()) {outputName=MouthAlpha, sortPriority=2}
  n5 = NamedRerouteUsage()
  n6 = NamedRerouteUsage()
  FunctionOutput(None=MaterialFunctionCall(UVs=Add(A=@n5, B=@n6), Texture Scale=NamedRerouteUsage()) {materialFunction=&ScaleUVsByCenter}.0-1 mask) {outputName=Eyes_Mask, sortPriority=2}
  n7 = NamedRerouteUsage()
  n8 = NamedRerouteUsage()
  FunctionOutput(None=MaterialFunctionCall(UVs=Add(A=@n7, B=@n8), Texture Scale=NamedRerouteUsage()) {materialFunction=&ScaleUVsByCenter}.0-1 mask) {outputName=Mouth_Mask, sortPriority=3}
  LinearInterpolate()
  n9 = TextureObjectParameter() {
   group=Flipbook 03 Eyes
   parameterName=Flipbook.Desaturated EyesTexture
   sortPriority=2
   texture=/Game/Global/Materials/Character_Parent_2023/FlipbookFace_Parent_2023/Textures/T_black_emptyalpha
  }
  NamedRerouteDeclaration(None=@n9) {name=Eyes_Flipbook_Desatured_Texture, nodeColor=<rgba 1 0 0.8438 1>}
  n10 = VectorParameter() {
   group=Flipbook 03 Eyes
   parameterName=Flipbook.EyesRemap_8-11
   sortPriority=202
   channelNames {a=7 - Unused, b=6 - Unused, g="9 - r_brow_in_dn_pose high value (Angry)", r="8 - r_brow_in_dn_pose low value (Bored)"}
   defaultValue=<rgba 8 9 10 11>
  }
  NamedRerouteDeclaration(None=@n10.RGBA) {name=Eyes_Remap_8-11, nodeColor=<rgba 1 0.6652 0.0014 1>}
  n11 = MaterialFunctionCall!(
   Curve Index=NamedRerouteUsage()
   Use Texture=NamedRerouteUsage()
   ----------------------
   Vertex Color=NamedRerouteUsage()
   ---------------------
   Flipbook Texture=NamedRerouteUsage()
   Column Count=NamedRerouteUsage()
   Row Count=NamedRerouteUsage()
   UVs=NamedRerouteUsage()
   UV Offset=NamedRerouteUsage()
   UV Scale=NamedRerouteUsage()
  ) {materialFunction=/Game/Global/Materials/Character_Parent_2023/FlipbookFace_Parent_2023/MaterialFunctions/MF_Sample_Flipbook}
  NamedRerouteDeclaration(None=@n11.Flipbook Alpha) {name=EyesAlpha_Desaturated, nodeColor=<rgba 0.2436 0 1 1>}
  NamedRerouteDeclaration(None=LinearInterpolate(A=NamedRerouteUsage(), B=NamedRerouteUsage(), Alpha=ScalarParameter() {group=Flipbook 01 Global, parameterName=FaceColorize})) {name=EyesColor, nodeColor=<rgba 0.2436 0 1 1>}
  NamedRerouteDeclaration(None=LinearInterpolate(A=NamedRerouteUsage(), B=NamedRerouteUsage(), Alpha=ScalarParameter() {group=Flipbook 01 Global, parameterName=FaceColorize})) {name=EyesAlpha, nodeColor=<rgba 0.2436 0 1 1>}
  NamedRerouteDeclaration(None=StaticBoolParameter() {defaultValue=true, group=Flipbook 01 Global, parameterName=Flipbook.UseEyes, sortPriority=-9}) {name=UseEyes, nodeColor=<rgba 0.1953 1 0 1>}
  NamedRerouteDeclaration(None=StaticBoolParameter() {defaultValue=true, group=Flipbook 01 Global, parameterName=Flipbook.UseMouth, sortPriority=-8}) {name=UseMouth, nodeColor=<rgba 0.1953 1 0 1>}
  n12 = MaterialFunctionCall(
   Curve Index=NamedRerouteUsage()
   Use Texture=NamedRerouteUsage()
   Vertex Color=NamedRerouteUsage()
   Flipbook Texture=NamedRerouteUsage()
   Column Count=NamedRerouteUsage()
   Row Count=NamedRerouteUsage()
   UVs=NamedRerouteUsage()
   UV Offset=NamedRerouteUsage()
   UV Scale=NamedRerouteUsage()
  ) {materialFunction=&MF_Sample_Flipbook}
  NamedRerouteDeclaration(None=@n12.Flipbook Vertex Mask) {name=MouthVertexMask, nodeColor=<rgba 0.2092 0.8266 1 1>}
  n13 = MaterialFunctionCall(
   Curve Index=NamedRerouteUsage()
   Use Texture=NamedRerouteUsage()
   Vertex Color=NamedRerouteUsage()
   Flipbook Texture=NamedRerouteUsage()
   Column Count=NamedRerouteUsage()
   Row Count=NamedRerouteUsage()
   UVs=NamedRerouteUsage()
   UV Offset=NamedRerouteUsage()
   UV Scale=NamedRerouteUsage()
  ) {materialFunction=&MF_Sample_Flipbook}
  NamedRerouteDeclaration(None=@n13.Flipbook Vertex Mask) {name=MouthVertexMask_Desaturated, nodeColor=<rgba 0.2092 0.8266 1 1>}
  NamedRerouteDeclaration(None=LinearInterpolate(A=NamedRerouteUsage(), B=NamedRerouteUsage(), Alpha=ScalarParameter() {group=Flipbook 01 Global, parameterName=FaceColorize})) {name=MouthColor, nodeColor=<rgba 0.2092 0.8266 1 1>}
  NamedRerouteDeclaration(None=LinearInterpolate(A=NamedRerouteUsage(), B=NamedRerouteUsage(), Alpha=ScalarParameter() {group=Flipbook 01 Global, parameterName=FaceColorize})) {name=MouthAlpha, nodeColor=<rgba 0.2092 0.8266 1 1>}
  n14 = Constant()
  n15 = Add(
   A=StaticSwitch(True=NamedRerouteUsage(), False=Constant(), Value=NamedRerouteUsage())
   B=StaticSwitch(True=NamedRerouteUsage(), False=Constant(), Value=NamedRerouteUsage())
  )
  n16 = StaticSwitch(True=NamedRerouteUsage(), False=Constant(), Value=NamedRerouteUsage())
  n17 = NamedRerouteUsage()
  n18 = Constant()
  n19 = NamedRerouteUsage()
  n20 = MaterialFunctionCall!(Vertex Mask=Add(A=Add(A=@n15, B=@n16), B=StaticSwitch(True=@n17, False=@n18, Value=@n19))) {materialFunction=/Game/Global/Materials/Character_Parent_2023/FlipbookFace_Parent_2023/MaterialFunctions/MF_Flipbook_VertColor_WPO}
  n21 = ScalarParameter() {
   controlType=Enumeration
   defaultValue=1
   enumeration=/Game/Global/Materials/Enums/E_Mat_OnOff
   group=Flipbook 01 Global
   parameterName=Flipbook.ApplyWPO
   sortPriority=1
  }
  NamedRerouteDeclaration(None=Add(A=LinearInterpolate(A=@n14, B=@n20.Flipbook WPO, Alpha=@n21), B=NamedRerouteUsage())) {name=Flipbook_WPO, nodeColor=<rgba 1 0 0.0938 1>}
  NamedRerouteDeclaration(None=@n20.Vertex Color Hide Mask) {name=VertexColorHideMask, nodeColor=<rgba 1 0.545 0.1681 1>}
  n22 = MaterialFunctionCall!(Material=NamedRerouteUsage(), Vertex Color Hide Mask=NamedRerouteUsage()) {materialFunction=/Game/Global/Materials/Character_Parent_2023/FlipbookFace_Parent_2023/MaterialFunctions/MF_Flipbook_Debug}
  NamedRerouteDeclaration(None=@n22.Debug Vert Color) {name=DebugVertColor, nodeColor=<rgba 0 0.6953 1 1>}
  NamedRerouteDeclaration(None=@n22.Debug Vert Color WPO) {name=DebugVertColor_WPO, nodeColor=<rgba 1 0.9844 0 1>}
  NamedRerouteDeclaration(None=@n22.Debug UV WPO) {name=DebugUV_WPO, nodeColor=<rgba 1 0 0.3047 1>}
  n23 = Add(
   A=StaticSwitch(True=NamedRerouteUsage(), False=Constant(), Value=NamedRerouteUsage())
   B=StaticSwitch(True=NamedRerouteUsage(), False=Constant(), Value=NamedRerouteUsage())
  )
  n24 = StaticSwitch(True=NamedRerouteUsage(), False=Constant(), Value=NamedRerouteUsage())
  NamedRerouteDeclaration(None=Saturate(None=Add(A=@n23, B=@n24))) {name=FaceMask, nodeColor=<rgba 1 0.7969 0 1>}
  n25 = StaticSwitch(True=NamedRerouteUsage(), False=Constant(), Value=NamedRerouteUsage())
  n26 = StaticSwitch(True=LinearInterpolate(A=@n25, B=NamedRerouteUsage(), Alpha=NamedRerouteUsage()), False=@n25, Value=NamedRerouteUsage())
  n27 = NamedRerouteUsage()
  n28 = NamedRerouteUsage()
  NamedRerouteDeclaration(None=StaticSwitch(True=LinearInterpolate(A=@n26, B=@n27, Alpha=@n28), False=@n26, Value=NamedRerouteUsage())) {name=FaceColor, nodeColor=<rgba 1 0.849 0.0986 1>}
  n29 = GetMaterialAttributes(None=NamedRerouteUsage()) {attributeGetTypes [
    #1
    #2
    #5
    #4
   ]}
  NamedRerouteDeclaration(None=@n29.BaseColor) {name=BaseColor, nodeColor=<rgba 0 1 0.5078 1>}
  NamedRerouteDeclaration(None=@n13.Flipbook Color) {name=MouthColor_Desaturated, nodeColor=<rgba 0.2092 0.8266 1 1>}
  NamedRerouteDeclaration(None=@n29.EmissiveColor) {name=Emissive, nodeColor=<rgba 1 0.1875 0 1>}
  NamedRerouteDeclaration(None=@n29.Opacity) {name=Opacity, nodeColor=<rgba 0 0.1328 1 1>}
  n30 = NamedRerouteUsage()
  n31 = NamedRerouteUsage()
  n32 = NamedRerouteUsage()
  n33 = ScalarParameter() {
   controlType=Enumeration
   defaultValue=1
   enumeration=&E_Mat_OnOff
   group=Flipbook 01 Global
   parameterName=Flipbook.ApplyBaseColor
   sortPriority=2
  }
  NamedRerouteDeclaration(None=LinearInterpolate(A=NamedRerouteUsage(), B=LinearInterpolate(A=@n30, B=@n31, Alpha=@n32), Alpha=@n33)) {name=Flipbook_BaseColor, nodeColor=<rgba 0.7578 1 0 1>}
  n34 = ScalarParameter() {
   controlType=Enumeration
   enumeration=&E_Mat_OnOff
   group=Flipbook 01 Global
   parameterName=Flipbook.ApplyOpacityMask
   sortPriority=4
  }
  NamedRerouteDeclaration(None=LinearInterpolate(A=NamedRerouteUsage(), B=NamedRerouteUsage(), Alpha=@n34)) {name=Flipbook_Opacity, nodeColor=<rgba 0 1 0.6016 1>}
  n35 = NamedRerouteUsage()
  n36 = Multiply(A=NamedRerouteUsage(), B=ScalarParameter() {defaultValue=1, group=Flipbook 01 Global, parameterName=Flipbook.EmissiveStrength})
  n37 = NamedRerouteUsage()
  n38 = ScalarParameter() {
   controlType=Enumeration
   enumeration=&E_Mat_OnOff
   group=Flipbook 01 Global
   parameterName=Flipbook.ApplyEmissive
   sortPriority=3
  }
  NamedRerouteDeclaration(None=LinearInterpolate(A=NamedRerouteUsage(), B=LinearInterpolate(A=@n35, B=@n36, Alpha=@n37), Alpha=@n38)) {name=Flipbook_Emissive, nodeColor=<rgba 0.108 0.2905 1 1>}
  NamedRerouteDeclaration(None=@n29.WorldPositionOffset) {name=WPO, nodeColor=<rgba 0.2275 0.2123 1 1>}
  n39 = StaticBoolParameter() {
   desc="Needed for when additional areas of the mesh use vertex colors for hiding geo. Avoiding conflict with the filpbooks vertex mask functionality. "
   group=Flipbook 01 Global
   parameterName=Flipbook.UseExtraVertexMask
   sortPriority=-7
  }
  NamedRerouteDeclaration(None=@n39) {name=UseExtraVertexMask, nodeColor=<rgba 0.1953 1 0 1>}
  NamedRerouteDeclaration(None=ChannelMaskParameter(None=VertexColor()) {group=Flipbook 01 Global, parameterName=Flipbook.ExtraVertexColor, sortPriority=-6}.R) {name=Extra_VertexColor, nodeColor=<rgba 1 0.9844 0 1>}
  NamedRerouteDeclaration(None=StaticBoolParameter() {group=Flipbook 06 Debug, parameterName=Flipbook.UseDebugVertColor, sortPriority=-1}) {name=UseDebugVertColor, nodeColor=<rgba 0.1953 1 0 1>}
  n40 = VectorParameter() {
   group=Flipbook 02 Brows
   parameterName=Flipbook.BrowsRemap_0-3
   sortPriority=201
   channelNames {a=3 - Brows Worried, b=2 - Brows Angry, g=1 - Brows Up, r=0 - Brows Neutral}
   defaultValue=<rgba 0 1 2 3>
  }
  NamedRerouteDeclaration(None=@n40.RGBA) {name=Brows_Remap_0-3, nodeColor=<rgba 1 0.6652 0.0014 1>}
  NamedRerouteDeclaration(None=StaticBoolParameter() {group=Flipbook 06 Debug, parameterName=Flipbook.UseDebugUVs, sortPriority=-1}) {name=UseDebugUVs, nodeColor=<rgba 0.1953 1 0 1>}
  n41 = VectorParameter() {
   group=Flipbook 04 Mouth
   parameterName=Flipbook.MouthRemap_4-7
   sortPriority=202
   channelNames {a=7 - Mouth Frown Open, b=6 - Mouth Frown Closed, g=5 - Mouth Smile Open, r=4 - Mouth Smile Closed}
   defaultValue=<rgba 4 5 6 7>
  }
  NamedRerouteDeclaration(None=@n41.RGBA) {name=Mouth_Remap_4-7, nodeColor=<rgba 1 0.6652 0.0014 1>}
  n42 = VectorParameter() {
   group=Flipbook 04 Mouth
   parameterName=Flipbook.MouthRemap_8-11
   sortPriority=203
   channelNames {a=11 - Oo Phoneme, b=10 - Ee Phoneme, g=9 - Mouth Grimmace Open, r=8 - Mouth Grimmace Closed}
   defaultValue=<rgba 8 9 10 11>
  }
  NamedRerouteDeclaration(None=@n42.RGBA) {name=Mouth_Remap_8-11, nodeColor=<rgba 1 0.6652 0.0014 1>}
  n43 = VectorParameter() {
   group=Flipbook 04 Mouth
   parameterName=Flipbook.MouthRemap_12-15
   sortPriority=204
   channelNames {a=15 - Unused, b=14 - Unused, g=13 - Unused, r=12 - Custom Mouth}
   defaultValue=<rgba 12 13 14 15>
  }
  NamedRerouteDeclaration(None=@n43.RGBA) {name=Mouth_Remap_12-15, nodeColor=<rgba 1 0.6652 0.0014 1>}
  n44 = VectorParameter() {
   group=Flipbook 02 Brows
   parameterName=Flipbook.BrowsRemap_4-7
   sortPriority=202
   channelNames {a=7 - Unused, b=6 - Unused, g=5 - Unused, r=4 - Custom Brows}
   defaultValue=<rgba 4 5 6 7>
  }
  NamedRerouteDeclaration(None=@n44.RGBA) {name=Brows_Remap_4-7, nodeColor=<rgba 1 0.6652 0.0014 1>}
  NamedRerouteDeclaration(None=@n13.Flipbook Alpha) {name=MouthAlpha_Desaturated, nodeColor=<rgba 0.2092 0.8266 1 1>}
  NamedRerouteDeclaration(None=ChannelMaskParameter(None=VertexColor()) {group=Flipbook 02 Brows, parameterName=Flipbook.BrowsVertexColor}.R) {name=Brows_VertexColor, nodeColor=<rgba 1 0.9844 0 1>}
  NamedRerouteDeclaration(None=StaticBoolParameter() {group=Flipbook 02 Brows, parameterName=Flipbook.UseTextureForBrows, sortPriority=-20}) {name=UseTextureForBrows, nodeColor=<rgba 0 1 0.3438 1>}
  n45 = TextureObjectParameter() {
   group=Flipbook 04 Mouth
   parameterName=Flipbook.Desaturated.MouthTexture
   sortPriority=2
   texture=&T_black_emptyalpha
  }
  NamedRerouteDeclaration(None=@n45) {name=Mouth_Flipbook_Deasaturated_Texture, nodeColor=<rgba 1 0.1875 0 1>}
  NamedRerouteDeclaration(None=FunctionInput() {inputType=FunctionInput_MaterialAttributes}) {name=Material, nodeColor=<rgba 1 0.7031 0 1>}
  n46 = TextureObjectParameter() {
   group=Flipbook 02 Brows
   parameterName=Flipbook.BrowsTexture
   sortPriority=2
   texture=&T_black_emptyalpha
  }
  NamedRerouteDeclaration(None=@n46) {name=Brows_FlipbookTexture, nodeColor=<rgba 1 0.1875 0 1>}
  NamedRerouteDeclaration(None=ScalarParameter() {defaultValue=3, group=Flipbook 02 Brows, parameterName=Flipbook.BrowsRowCount, sortPriority=3}) {name=Brows_RowCount, nodeColor=<rgba 1 0 0.7734 1>}
  NamedRerouteDeclaration(None=ScalarParameter() {defaultValue=3, group=Flipbook 02 Brows, parameterName=Flipbook.BrowsColumnCount, sortPriority=4}) {name=Brows_ColumnCount, nodeColor=<rgba 1 0 0.7734 1>}
  NamedRerouteDeclaration(None=StaticSwitchParameter(True=TextureCoordinate() {coordinateIndex=1}, False=TextureCoordinate()) {group=Flipbook 02 Brows, parameterName=Flipbook.BrowsUseUV2, sortPriority=10}) {name=Brows_UVs, nodeColor=<rgba 1 0.5859 0 1>}
  n47 = ScalarParameter() {defaultValue=1, group=Flipbook 02 Brows, parameterName=Flipbook.BrowsUVScaleX, sortPriority=12}
  n48 = ScalarParameter() {defaultValue=1, group=Flipbook 02 Brows, parameterName=Flipbook.BrowsUVScaleY, sortPriority=13}
  NamedRerouteDeclaration(None=Multiply(
    A=ScalarParameter() {defaultValue=1, group=Flipbook 02 Brows, parameterName=Flipbook.BrowsGlobalScale, sortPriority=12}
    B=AppendVector(A=@n47, B=@n48)
   )) {name=Brows_UVScale, nodeColor=<rgba 0.2656 0 1 1>}
  NamedRerouteDeclaration(None=AppendVector(
    A=ScalarParameter() {group=Flipbook 02 Brows, parameterName=Flipbook.BrowsUVOffsetX, sortPriority=20}
    B=ScalarParameter() {group=Flipbook 02 Brows, parameterName=Flipbook.BrowsUVOffsetY, sortPriority=21}
   )) {name=Brows_UVOffset, nodeColor=<rgba 0.0547 1 0 1>}
  NamedRerouteDeclaration(None=MaterialFunctionCall!(
    Index=MaterialFunctionCall() {materialFunction=/CosmeticCompanions/Assets/Biped/TapiTime/Materials/Dev_Flipbook/Material_Functions/MF_SK_Flipbook_Brows_Curves_Dev}.Index
    Input_0-3=NamedRerouteUsage()
    Input_4-7=NamedRerouteUsage()
   ) {materialFunction=/Game/Global/Materials/Character_Parent_2023/FlipbookFace_Parent_2023/MaterialFunctions/MF_RemapIndex_8}.Remapped Index) {name=Brows_CurveIndex, nodeColor=<rgba 1 0.4453 0 1>}
  n49 = VectorParameter() {
   group=Flipbook 03 Eyes
   parameterName=Flipbook.EyesRemap_0-3
   sortPriority=201
   channelNames {a=3 - r_lower_lid_up, b=2 - Happy Blink, g=1 - Blink, r=0 - Neutral}
   defaultValue=<rgba 0 1 2 3>
  }
  NamedRerouteDeclaration(None=@n49.RGBA) {name=Eyes_Remap_0-3, nodeColor=<rgba 1 0.6652 0.0014 1>}
  n50 = VectorParameter() {
   group=Flipbook 03 Eyes
   parameterName=Flipbook.EyesRemap_4-7
   sortPriority=202
   channelNames {a="7 - r_brow_in_pose (Worried)", b=6 - Upper Lid Down, g=5 - Eye Emote 01, r=4 - Eyes Wide}
   defaultValue=<rgba 8 9 10 11>
  }
  NamedRerouteDeclaration(None=@n50.RGBA) {name=Eyes_Remap_4-7, nodeColor=<rgba 1 0.6652 0.0014 1>}
  NamedRerouteDeclaration(None=@n12.Flipbook Color) {name=MouthColor_Base, nodeColor=<rgba 0.2092 0.8266 1 1>}
  NamedRerouteDeclaration(None=@n12.Flipbook Alpha) {name=MouthAlpha_Base, nodeColor=<rgba 0.2092 0.8266 1 1>}
  n51 = MaterialFunctionCall(
   Curve Index=NamedRerouteUsage()
   Use Texture=NamedRerouteUsage()
   Vertex Color=NamedRerouteUsage()
   Flipbook Texture=NamedRerouteUsage()
   Column Count=NamedRerouteUsage()
   Row Count=NamedRerouteUsage()
   UVs=NamedRerouteUsage()
   UV Offset=NamedRerouteUsage()
   UV Scale=NamedRerouteUsage()
  ) {materialFunction=&MF_Sample_Flipbook}
  NamedRerouteDeclaration(None=@n51.Flipbook Color) {name=EyesColor_Base, nodeColor=<rgba 0.2436 0 1 1>}
  NamedRerouteDeclaration(None=@n51.Flipbook Alpha) {name=EyesAlpha_Base, nodeColor=<rgba 0.2436 0 1 1>}
  NamedRerouteDeclaration(None=MaterialFunctionCall!(
    Index=MaterialFunctionCall() {materialFunction=/CosmeticCompanions/Assets/Biped/TapiTime/Materials/Dev_Flipbook/Material_Functions/MF_SK_Flipbook_Eyes_Curves_Dev}.Index
    Input_0-3=NamedRerouteUsage()
    Input_4-7=NamedRerouteUsage()
    Input_8-11=NamedRerouteUsage()
   ) {materialFunction=/Game/Global/Materials/Character_Parent_2023/FlipbookFace_Parent_2023/MaterialFunctions/MF_RemapIndex_12}.Remapped Index) {name=Eyes_CurveIndex, nodeColor=<rgba 1 0.4453 0 1>}
  NamedRerouteDeclaration(None=StaticBoolParameter() {group=Flipbook 03 Eyes, parameterName=Flipbook.UseTextureForEyes, sortPriority=-20}) {name=UseTextureForEyes, nodeColor=<rgba 0 1 0.3438 1>}
  NamedRerouteDeclaration(None=ScalarParameter() {defaultValue=3, group=Flipbook 03 Eyes, parameterName=Flipbook.EyesRowCount, sortPriority=3}) {name=Eyes_RowCount, nodeColor=<rgba 1 0 0.7734 1>}
  NamedRerouteDeclaration(None=ScalarParameter() {defaultValue=3, group=Flipbook 03 Eyes, parameterName=Flipbook.EyesColumnCount, sortPriority=4}) {name=Eyes_ColumnCount, nodeColor=<rgba 1 0 0.7734 1>}
  NamedRerouteDeclaration(None=StaticSwitchParameter(True=TextureCoordinate() {coordinateIndex=1}, False=TextureCoordinate()) {group=Flipbook 03 Eyes, parameterName=Flipbook.EyesUseUV2, sortPriority=10}) {name=Eyes_UVs, nodeColor=<rgba 1 0.5859 0 1>}
  n52 = ScalarParameter() {defaultValue=1, group=Flipbook 03 Eyes, parameterName=Flipbook.EyesUVScaleX, sortPriority=12}
  n53 = ScalarParameter() {defaultValue=1, group=Flipbook 03 Eyes, parameterName=Flipbook.EyesUVScaleY, sortPriority=13}
  NamedRerouteDeclaration(None=Multiply(
    A=ScalarParameter() {defaultValue=1, group=Flipbook 03 Eyes, parameterName=Flipbook.EyesGlobalScale, sortPriority=12}
    B=AppendVector(A=@n52, B=@n53)
   )) {name=Eyes_UVScale, nodeColor=<rgba 0.2656 0 1 1>}
  NamedRerouteDeclaration(None=AppendVector(
    A=ScalarParameter() {group=Flipbook 03 Eyes, parameterName=Flipbook.EyesUVOffsetX, sortPriority=20}
    B=ScalarParameter() {group=Flipbook 03 Eyes, parameterName=Flipbook.EyesUVOffsetY, sortPriority=21}
   )) {name=Eyes_UVOffset, nodeColor=<rgba 0.0547 1 0 1>}
  NamedRerouteDeclaration(None=ChannelMaskParameter(None=VertexColor()) {group=Flipbook 03 Eyes, parameterName=Flipbook.EyesVertexColor}.R) {name=Eyes_VertexColor, nodeColor=<rgba 1 0.9844 0 1>}
  n54 = TextureObjectParameter() {
   group=Flipbook 03 Eyes
   parameterName=Flipbook.EyesTexture
   sortPriority=2
   texture=&T_black_emptyalpha
  }
  NamedRerouteDeclaration(None=@n54) {name=Eyes_FlipbookTexture, nodeColor=<rgba 1 0.1875 0 1>}
  n55 = VectorParameter() {
   group=Flipbook 04 Mouth
   parameterName=Flipbook.MouthRemap_0-3
   sortPriority=201
   channelNames {a=3 - Mouth Press, b=2 - Mouth Neutral Wide, g=1 - Mouth Neutral Open, r=0 - Mouth Neutral Closed}
   defaultValue=<rgba 0 1 2 3>
  }
  NamedRerouteDeclaration(None=@n55.RGBA) {name=Mouth_Remap_0-3, nodeColor=<rgba 1 0.6652 0.0014 1>}
  NamedRerouteDeclaration(None=MaterialFunctionCall!(
    Index=MaterialFunctionCall() {materialFunction=/CosmeticCompanions/Assets/Biped/TapiTime/Materials/Dev_Flipbook/Material_Functions/MF_SK_Flipbook_Mouth_Curves_Dev}.Index
    Input_0-3=NamedRerouteUsage()
    Input_4-7=NamedRerouteUsage()
    Input_8-11=NamedRerouteUsage()
    Input_12-15=NamedRerouteUsage()
   ) {materialFunction=/Game/Global/Materials/Character_Parent_2023/FlipbookFace_Parent_2023/MaterialFunctions/MF_RemapIndex_16}.Remapped Index) {name=Mouth_CurveIndex, nodeColor=<rgba 1 0.4453 0 1>}
  NamedRerouteDeclaration(None=ChannelMaskParameter(None=VertexColor()) {group=Flipbook 04 Mouth, parameterName=Flipbook.MouthVertexColor}.R) {name=Mouth_VertexColor, nodeColor=<rgba 1 0.9844 0 1>}
  NamedRerouteDeclaration(None=StaticBoolParameter() {defaultValue=true, group=Flipbook 04 Mouth, parameterName=Flipbook.UseTextureForMouth, sortPriority=-20}) {name=UseTextureForMouth, nodeColor=<rgba 0 1 0.3438 1>}
  n56 = TextureObjectParameter() {
   group=Flipbook 04 Mouth
   parameterName=Flipbook.MouthTexture
   sortPriority=2
   texture=&T_black_emptyalpha
  }
  NamedRerouteDeclaration(None=@n56) {name=Mouth_FlipbookTexture, nodeColor=<rgba 1 0.1875 0 1>}
  NamedRerouteDeclaration(None=ScalarParameter() {defaultValue=3, group=Flipbook 04 Mouth, parameterName=Flipbook.MouthRowCount, sortPriority=3}) {name=Mouth_RowCount, nodeColor=<rgba 1 0 0.7734 1>}
  NamedRerouteDeclaration(None=ScalarParameter() {defaultValue=3, group=Flipbook 04 Mouth, parameterName=Flipbook.MouthColumnCount, sortPriority=4}) {name=Mouth_ColumnCount, nodeColor=<rgba 1 0 0.7734 1>}
  NamedRerouteDeclaration(None=StaticSwitchParameter(True=TextureCoordinate() {coordinateIndex=1}, False=TextureCoordinate()) {group=Flipbook 04 Mouth, parameterName=Flipbook.MouthUseUV2, sortPriority=10}) {name=Mouth_UVs, nodeColor=<rgba 1 0.5859 0 1>}
  n57 = ScalarParameter() {defaultValue=1, group=Flipbook 04 Mouth, parameterName=Flipbook.MouthUVScaleX, sortPriority=12}
  n58 = ScalarParameter() {defaultValue=1, group=Flipbook 04 Mouth, parameterName=Flipbook.MouthUVScaleY, sortPriority=13}
  NamedRerouteDeclaration(None=Multiply(
    A=ScalarParameter() {defaultValue=1, group=Flipbook 04 Mouth, parameterName=Flipbook.MouthGlobalScale, sortPriority=12}
    B=AppendVector(A=@n57, B=@n58)
   )) {name=Mouth_UVScale, nodeColor=<rgba 0.2656 0 1 1>}
  NamedRerouteDeclaration(None=AppendVector(
    A=ScalarParameter() {group=Flipbook 04 Mouth, parameterName=Flipbook.MouthUVOffsetX, sortPriority=20}
    B=ScalarParameter() {group=Flipbook 04 Mouth, parameterName=Flipbook.MouthUVOffsetY, sortPriority=21}
   )) {name=Mouth_UVOffset, nodeColor=<rgba 0.0547 1 0 1>}
  NamedRerouteDeclaration(None=@n11.Flipbook Color) {name=EyesColor_Desaturated, nodeColor=<rgba 0.2436 0 1 1>}
  n59 = Constant() {r=1}
  n60 = NamedRerouteUsage()
  n61 = ScalarParameter() {defaultValue=0.1, group=Flipbook 01 Global, parameterName=Flipbook.ExtraVertexColorValue, sortPriority=-6}
  NamedRerouteDeclaration(None=LinearInterpolate(
    A=MaterialFunctionCall!(Hide Element Toggle=@n59, VertexColorChannel=@n60, Vertex Value to hide=@n61) {materialFunction=/Game/Packages/Fortress_SharedMaterials/Base_Material_Functions/VertexColorMask}.Result
    B=Constant()
    Alpha=ScalarParameter() {desc=Set by CID, group=Flipbook 01 Global, parameterName=Flipbook.HideExtraVertexColor, sortPriority=-5}
   )) {name=ExtraVertexMask, nodeColor=<rgba 0 1 0.3906 1>}
  NamedRerouteDeclaration(None=StaticBoolParameter() {group=Flipbook 01 Global, parameterName=Flipbook.UseBrows, sortPriority=-10}) {name=UseBrows, nodeColor=<rgba 0.1953 1 0 1>}
  n62 = MaterialFunctionCall(
   Curve Index=NamedRerouteUsage()
   Use Texture=NamedRerouteUsage()
   Vertex Color=NamedRerouteUsage()
   Flipbook Texture=NamedRerouteUsage()
   Column Count=NamedRerouteUsage()
   Row Count=NamedRerouteUsage()
   UVs=NamedRerouteUsage()
   UV Offset=NamedRerouteUsage()
   UV Scale=NamedRerouteUsage()
  ) {materialFunction=&MF_Sample_Flipbook}
  NamedRerouteDeclaration(None=@n62.Flipbook Vertex Mask) {name=BrowsVertexMask, nodeColor=<rgba 1 0 0.8438 1>}
  NamedRerouteDeclaration(None=@n62.Flipbook Color) {name=BrowsColor, nodeColor=<rgba 1 0 0.8438 1>}
  NamedRerouteDeclaration(None=@n62.Flipbook Alpha) {name=BrowsAlpha, nodeColor=<rgba 1 0 0.8438 1>}
  NamedRerouteDeclaration(None=@n51.Flipbook Vertex Mask) {name=EyesVertexMask, nodeColor=<rgba 0.2436 0 1 1>}
  ScalarParameter() {group=Flipbook 01 Global, parameterName=FaceColorize}
 ]}
