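# Skeletal-mesh body material for the TapiTime biped cosmetic companion, with custom color, liquid, fake-glass and animated-face looks.
# Character parent PreFX feeds mask-driven custom color, liquid, fake-glass and flipbook functions, then PostFX.
# Face tint recolors the flipbook eyes and mouth, while a masked Fresnel darkens edges through base color, emissive and roughness.
# Eye and mouth change parameters drive temporal responsiveness output.
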Material /CosmeticCompanions/Assets/Biped/TapiTime/Materials/M_TapiTime_Body
graph {
 properties {
  bAutomaticallySetUsageInEditor=false
  bUseMaterialAttributes=true
  bUsedWithMeshDeformer=true
  bUsedWithMorphTargets=true
  bUsedWithSkeletalMesh=true
  bUsedWithStaticLighting=true
  previewMesh=/BRCosmetics/Accessories/FORT_Backpacks/Backpack_M_MED_BlobRock/Meshes/M_MED_BlobRock_Pack
  shadingModel=MSM_FromMaterialExpression
  usedShadingModels=Default Lit | Subsurface | Clear Coat | Subsurface Profile | Cloth
 }
 nodes [
  n1 = NamedRerouteUsage()
  n2 = Saturate(None=Multiply(A=SmoothStep(Value=@n1) {constMin=-0.3}) {constB=2})
  n3 = Constant() {r=1}
  n4 = NamedRerouteUsage()
  LinearInterpolate(
   A=Multiply(
    A=Desaturation(None=@n2, Fraction=@n3)
    B=VectorParameter() {group=Flipbook 01 Global, parameterName=FaceColor, defaultValue=<rgba 0.2617 0.0503 0.5365 0>}.RGB
   )
   B=Constant3Vector() {constant=<rgba 0.9 0.9 0.9 0>}
   Alpha=Multiply(A=ComponentMask(None=@n4) {g=true}, B=NamedRerouteUsage())
  )
  n5 = SmoothStep(
   Min=ScalarParameter() {group=FakeGlass, parameterName=MinSmoothstepDarkenEdges}
   Max=ScalarParameter() {defaultValue=1, group=FakeGlass, parameterName=MaxSmoothstepDarkenEdges}
   Value=Fresnel(ExponentIn=ScalarParameter() {defaultValue=0.5, group=FakeGlass, parameterName=ExponentInDarkenEdges})
  )
  n6 = NamedRerouteUsage()
  NamedRerouteDeclaration(None=NamedRerouteDeclaration(None=Multiply(A=@n5, B=@n6)) {name=DarkenRoughnessMask, nodeColor=<rgba 0.8516 1 0 1>}) {name=DarkenEdgesMask, nodeColor=<rgba 1 0 0.6094 1>}
  n7 = MaterialFunctionCall() {materialFunction=&MF_CharacterParent_2023_PreFX}
  n8 = TextureSampleParameter2D() {
   parameterName=CompanionCustomColorMask
   samplerType=SAMPLERTYPE_Masks
   texture=/CosmeticCompanions/Assets/Biped/TapiTime/Textures/T_TapiTime_FX2
  }
  n9 = MaterialFunctionCall!(Material=@n7.Result, Tech Art Mask=@n8.RGB) {materialFunction=&/MF_FN_TapiTime_Custom_Color}
  n10 = TextureSampleParameter2D() {
   parameterName=TextureMasksForBoba
   samplerType=SAMPLERTYPE_Masks
   texture=/CosmeticCompanions/Assets/Biped/TapiTime/Textures/T_TapiTime_Body_FX2
  }
  n11 = NamedRerouteDeclaration(None=@n10.RGB) {name=TextureMasksForBoba, nodeColor=<rgba 0.6875 1 0 1>}
  n12 = MaterialFunctionCall!(In=MaterialFunctionCall!(
    Material Attributes=MaterialFunctionCall!(Mat=@n9.Base Color, TechArtMask=@n11) {materialFunction=&/MF_TapiTime_Liquid}.Result
    Fresnel=@n7.Fresnel
    TechArtMask=@n8.RGB
   ) {materialFunction=&/MF_FakeGlass_TapiTime}.Result) {materialFunction=&/MF_TapiTime_SK_Flipbook_Base}
  n13 = GetMaterialAttributes(None=@n12.Result) {attributeGetTypes [#1]}
  n14 = NamedRerouteUsage()
  n15 = GetMaterialAttributes(None=MaterialFunctionCall(
    Material=SetMaterialAttributes!(MaterialAttributes=@n13.MaterialAttributes, Base Color=@n14) {attributeSetTypes [#1]}
    M=@n7.M
    Pre Skinned Local Position=@n7.M
    Distance Blend=@n7.M
    Fresnel=@n7.M
   ) {materialFunction=&MF_CharacterParent_2023_PostFX}.Result) {attributeGetTypes [#2, #1, #3]}
  NamedRerouteDeclaration(None=@n15.Emissive Color) {name=EmissiveBeforeDarkenEdges, nodeColor=<rgba 0 0.9297 1 1>}
  NamedRerouteDeclaration(None=Add(A=NamedRerouteUsage(), B=NamedRerouteUsage())) {name=FaceMask, nodeColor=<rgba 0.6172 1 0 1>}
  NamedRerouteDeclaration(None=Add(A=NamedRerouteUsage(), B=NamedRerouteUsage())) {name=FaceAlpha, nodeColor=<rgba 0.7812 0 1 1>}
  NamedRerouteDeclaration(None=@n13.Base Color) {name=BaseColorAfterFlipbookFace, nodeColor=<rgba 0.8984 1 0 1>}
  n16 = NamedRerouteUsage()
  n17 = Multiply(
   A=ComponentMask(None=NamedRerouteUsage()) {r=true}
   B=VectorParameter() {group=Flipbook 01 Global, parameterName=FaceColor, defaultValue=<rgba 0.2617 0.0503 0.5365 0>}.RGB
  )
  n18 = ScalarParameter() {group=Flipbook 01 Global, parameterName=FaceColorize}
  n19 = ComponentMask(None=NamedRerouteUsage()) {g=true}
  NamedRerouteDeclaration(None=LinearInterpolate(A=LinearInterpolate(A=@n16, B=@n17, Alpha=@n18), B=@n19, Alpha=@n19)) {name=FaceHSVShifted, nodeColor=<rgba 0 1 0.4609 1>}
  NamedRerouteDeclaration(None=LinearInterpolate(A=NamedRerouteUsage(), B=NamedRerouteUsage(), Alpha=NamedRerouteUsage())) {name=BaseColorWithHSVShiftedFace, nodeColor=<rgba 1 0.1406 0 1>}
  NamedRerouteDeclaration(None=@n12.EyesAlpha) {name=EyeAlpha, nodeColor=<rgba 1 0 0.9141 1>}
  NamedRerouteDeclaration(None=@n12.Eyes_Mask) {name=Eye_Mask, nodeColor=<rgba 0.002 1 0.56 1>}
  NamedRerouteDeclaration(None=@n12.MouthAlpha) {name=MouthAlpha, nodeColor=<rgba 1 0.4453 0 1>}
  NamedRerouteDeclaration(None=@n15.Base Color) {name=BaseColorBeforeDarkenEdges, nodeColor=<rgba 1 0.75 0 1>}
  NamedRerouteDeclaration(None=@n12.Mouth_Mask) {name=Mouth_Mask, nodeColor=<rgba 0.087 0.2997 1 1>}
  NamedRerouteDeclaration(None=@n15.Roughness) {name=RoughnessBeforeDarkenEdges, nodeColor=<rgba 0.4297 0 1 1>}
  NamedRerouteDeclaration(None=LinearInterpolate(A=NamedRerouteUsage(), B=NamedRerouteUsage(), Alpha=NamedRerouteUsage())) {name=FinalEmissiveColor, nodeColor=<rgba 0 1 0.1094 1>}
  NamedRerouteDeclaration(None=LinearInterpolate(A=NamedRerouteUsage(), B=NamedRerouteUsage(), Alpha=NamedRerouteUsage())) {name=FinalBaseColor, nodeColor=<rgba 1 0 0.2109 1>}
  NamedRerouteDeclaration(None=LinearInterpolate(A=NamedRerouteUsage(), Alpha=NamedRerouteUsage())) {name=FinalRoughness, nodeColor=<rgba 0 0.5312 1 1>}
  NamedRerouteDeclaration(None=VectorParameter() {parameterName=DarkenEdgeColor, defaultValue=<rgba 0.0386 0.0106 0.0833 1>}.RGB) {name=DarkenEdgeColor, nodeColor=<rgba 0 0.8594 1 1>}
  NamedRerouteDeclaration(None=ChannelMaskParameter(None=NamedRerouteUsage()) {group=Liquid, maskChannel=Green, parameterName=DarkenEdgesAreaMask, defaultValue=<rgba 0 1 0 0>}.R) {name=DarkenEdgesAreaMask, nodeColor=<rgba 1 0 0.375 1>}
  NamedRerouteUsage()
  SetMaterialAttributes!(
   MaterialAttributes=@n15.MaterialAttributes
   Emissive Color=LinearInterpolate(A=NamedRerouteUsage(), Alpha=NamedRerouteUsage()) {constB=0}
   Base Color=LinearInterpolate(A=NamedRerouteUsage(), B=NamedRerouteUsage(), Alpha=NamedRerouteUsage())
   Roughness=LinearInterpolate(A=NamedRerouteUsage(), Alpha=NamedRerouteUsage())
  ) {attributeSetTypes [#2, #1, #3]}
  n20 = Max(
   A=Multiply(A=ScalarParameter() {parameterName=HasEyesChanged}, B=NamedRerouteUsage())
   B=Multiply(A=ScalarParameter() {parameterName=HasMouthChanged}, B=NamedRerouteUsage())
  )
  TemporalResponsivenessOutput!(None=Saturate(None=LinearInterpolate(Alpha=@n20) {constA=0.4}))
 ]
}
