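# Animation Blueprint for the TapiTime biped cosmetic companion that retargets a pose from another mesh and flags facial curve activity to its materials.
# The AnimGraph uses bUseReverseRetarget to pick the emote or standard IK Retargeter.
# Each update compares eye and mouth curves to cached values within a threshold.
# HasEyesChanged and HasMouthChanged receive the OR of those nearly-equal tests, uninverted, and both functions share the cached smile value.
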AnimBlueprint /CosmeticCompanions/Assets/Biped/TapiTime/Meshes/TapiTime_Biped_AnimBP
target_skeleton=&/TapiTime_Biped_Skeleton
graphs [
 AnimGraph {nodes [
   Copy Pose From Mesh() {bCopyCurves=true, bUseAttachedParent=true}
   Output Pose(Result=Blend Poses by bool(
     BlendPose_0=Retarget Pose From Mesh!() -> (Pose:Pose Link Structure) {is=title AnimGraphNode_RetargetPoseFromMesh, iKRetargeterAsset=&/RTG_TapiTime_EmoteRTG}
     BlendPose_1=Retarget Pose From Mesh() {iKRetargeterAsset=&/RTG_TapiTime}
     bActiveValue=Get bUseReverseRetarget() -> (bUseReverseRetarget:bool)
    ))
  ]}
 EventGraph {nodes [
   n1 = Event BlueprintUpdateAnimation()
   > Parent: BlueprintUpdateAnimation!(execute, DeltaTimeX:float=@n1.DeltaTimeX) -> (then) {is=title K2Node_CallParentFunction}
   > NotifyMaterialOfEyeUpdates!(execute, self, Threshold:double=0.2) -> (then)
   > NotifyMaterialOfMouthUpdates!(execute, self, Threshold:double=0.2) -> (then)
   TryGetPawnOwner()
  ]}
 NotifyMaterialOfEyeUpdates {nodes [
   n1 = NotifyMaterialOfEyeUpdates!() -> (then, Threshold:double=~) {is=title K2Node_FunctionEntry}
   > Set Threshold(Threshold:double=@n1) -> (Output_Get:double)
   n2 = GetCurveValue(CurveName=r_blink_pose)
   n3 = "Nearly Equal (Float)"(A=@n2, B=Get PreviousState_r_blink_pose() -> (PreviousState_r_blink_pose:double), ErrorTolerance=Get Threshold() -> (Threshold:double))
   n4 = GetCurveValue(CurveName=r_smile_pose)
   n5 = "Nearly Equal (Float)"(A=@n4, B=Get PreviousState_r_smile_pose() -> (PreviousState_r_smile_pose:double), ErrorTolerance=Get Threshold())
   n6 = GetCurveValue(CurveName=r_wide_pose)
   n7 = "Nearly Equal (Float)"(A=@n6, B=Get PreviousState_r_wide_pose() -> (PreviousState_r_wide_pose:double), ErrorTolerance=Get Threshold())
   n8 = GetCurveValue(CurveName=r_brow_in_dn_pose)
   n9 = "Nearly Equal (Float)"(A=@n8, B=Get PreviousState_r_brow_in_dn_pose() -> (PreviousState_r_brow_in_dn_pose:double), ErrorTolerance=Get Threshold())
   n10 = GetCurveValue(CurveName=r_lower_lid_up_pose)
   n11 = "Nearly Equal (Float)"(A=@n10, B=Get PreviousState_r_lower_lid_up_pose() -> (PreviousState_r_lower_lid_up_pose:double), ErrorTolerance=Get Threshold())
   n12 = GetCurveValue(CurveName=r_upperliddown_pose)
   n13 = "Nearly Equal (Float)"(A=@n12, B=Get PreviousState_r_upperliddown_pose() -> (PreviousState_r_upperliddown_pose:double), ErrorTolerance=Get Threshold())
   n14 = GetCurveValue(CurveName=Eye_Emote_01)
   n15 = "Nearly Equal (Float)"(A=@n14, B=Get PreviousState_Eye_Emote_01() -> (PreviousState_Eye_Emote_01:double), ErrorTolerance=Get Threshold())
   n16 = GetCurveValue(CurveName=r_brow_in_up_pose)
   n17 = "Nearly Equal (Float)"(A=@n16, B=Get PreviousState_r_brow_in_up_pose() -> (PreviousState_r_brow_in_up_pose:double), ErrorTolerance=Get Threshold())
   > SetScalarParameterValueOnMaterials!(
    execute
    self:Mesh Component Object Reference=GetOwningComponent()
    ParameterName:Name=HasEyesChanged
    ParameterValue:float="To Float (Boolean)"(InBool=OR Boolean!(A=@n3, B=@n5, C:bool=@n7, D:bool=@n11, E:bool=@n13, F:bool=@n15, G:bool=@n17, H:bool=@n9) -> (ReturnValue))
   ) -> (then)
   > Set PreviousState_r_blink_pose(PreviousState_r_blink_pose:double=@n2) -> (Output_Get:double)
   > Set PreviousState_r_smile_pose(PreviousState_r_smile_pose:double=@n4) -> (Output_Get:double)
   > Set PreviousState_r_wide_pose(PreviousState_r_wide_pose:double=@n6) -> (Output_Get:double)
   > Set PreviousState_r_lower_lid_up_pose(PreviousState_r_lower_lid_up_pose:double=@n10) -> (Output_Get:double)
   > Set PreviousState_r_upperliddown_pose(PreviousState_r_upperliddown_pose:double=@n12) -> (Output_Get:double)
   > Set PreviousState_Eye_Emote_01(PreviousState_Eye_Emote_01:double=@n14) -> (Output_Get:double)
   > Set PreviousState_r_brow_in_up_pose(PreviousState_r_brow_in_up_pose:double=@n16) -> (Output_Get:double)
   > Set PreviousState_r_brow_in_dn_pose(PreviousState_r_brow_in_dn_pose:double=@n8) -> (Output_Get:double)
   > Return Node()
  ]}
 NotifyMaterialOfMouthUpdates {nodes [
   n1 = NotifyMaterialOfMouthUpdates!() -> (then, Threshold:double=~) {is=title K2Node_FunctionEntry}
   > Set Threshold(Threshold=@n1)
   n2 = GetCurveValue(CurveName=r_smile_pose)
   n3 = "Nearly Equal (Float)"(A=@n2, B=Get PreviousState_r_smile_pose(), ErrorTolerance=Get Threshold())
   n4 = GetCurveValue(CurveName=r_frown_pose)
   n5 = "Nearly Equal (Float)"(A=@n4, B=Get PreviousState_r_frown_pose() -> (PreviousState_r_frown_pose:double), ErrorTolerance=Get Threshold())
   n6 = GetCurveValue(CurveName=jaw_open_pose)
   n7 = "Nearly Equal (Float)"(A=@n6, B=Get PreviousState_jaw_open_pose() -> (PreviousState_jaw_open_pose:double), ErrorTolerance=Get Threshold())
   n8 = GetCurveValue(CurveName=mouth_emote_01)
   n9 = "Nearly Equal (Float)"(A=@n8, B=Get PreviousState_mouth_emote_01() -> (PreviousState_mouth_emote_01:double), ErrorTolerance=Get Threshold())
   n10 = GetCurveValue(CurveName=reverse_jaw_pose)
   n11 = "Nearly Equal (Float)"(A=@n10, B=Get PreviousState_reverse_jaw_pose() -> (PreviousState_reverse_jaw_pose:double), ErrorTolerance=Get Threshold())
   n12 = GetCurveValue(CurveName=r_lip_corner_wide_pose)
   n13 = "Nearly Equal (Float)"(A=@n12, B=Get PreviousState_r_lip_corner_wide_pose() -> (PreviousState_r_lip_corner_wide_pose:double), ErrorTolerance=Get Threshold())
   n14 = GetCurveValue(CurveName=mouth_emote_Ee)
   n15 = "Nearly Equal (Float)"(A=@n14, B=Get PreviousState_mouth_emote_Ee() -> (PreviousState_mouth_emote_Ee:double), ErrorTolerance=Get Threshold())
   n16 = GetCurveValue(CurveName=mouth_emote_Oo)
   n17 = "Nearly Equal (Float)"(A=@n16, B=Get PreviousState_mouth_emote_Oo() -> (PreviousState_mouth_emote_Oo:double), ErrorTolerance=Get Threshold())
   n18 = GetCurveValue(CurveName=jaw_up_pose)
   n19 = "Nearly Equal (Float)"(A=@n18, B=Get PreviousState_jaw_up_pose() -> (PreviousState_jaw_up_pose:double), ErrorTolerance=Get Threshold())
   n20 = GetCurveValue(CurveName=r_lower_lip_down_pose)
   n21 = "Nearly Equal (Float)"(A=@n20, B=Get PreviousState_r_lower_lip_down_pose() -> (PreviousState_r_lower_lip_down_pose:double), ErrorTolerance=Get Threshold())
   n22 = GetCurveValue(CurveName=r_upper_lip_raiser_pose)
   n23 = "Nearly Equal (Float)"(A=@n22, B=Get PreviousState_r_upper_lip_raiser_pose() -> (PreviousState_r_upper_lip_raiser_pose:double), ErrorTolerance=Get Threshold())
   > SetScalarParameterValueOnMaterials(
    self=GetOwningComponent()
    ParameterName=HasMouthChanged
    ParameterValue="To Float (Boolean)"(InBool=OR Boolean!(A=@n3, B=@n5, C=@n7, D=@n11, E=@n13, F=@n19, G=@n21, H=@n23, I:bool=@n15, J:bool=@n17, K:bool=@n9) -> (ReturnValue))
   )
   > Set PreviousState_r_smile_pose(PreviousState_r_smile_pose=@n2)
   > Set PreviousState_r_frown_pose(PreviousState_r_frown_pose:double=@n4) -> (Output_Get:double)
   > Set PreviousState_jaw_open_pose(PreviousState_jaw_open_pose:double=@n6) -> (Output_Get:double)
   > Set PreviousState_reverse_jaw_pose(PreviousState_reverse_jaw_pose:double=@n10) -> (Output_Get:double)
   > Set PreviousState_r_lip_corner_wide_pose(PreviousState_r_lip_corner_wide_pose:double=@n12) -> (Output_Get:double)
   > Set PreviousState_jaw_up_pose(PreviousState_jaw_up_pose:double=@n18) -> (Output_Get:double)
   > Set PreviousState_r_lower_lip_down_pose(PreviousState_r_lower_lip_down_pose:double=@n20) -> (Output_Get:double)
   > Set PreviousState_r_upper_lip_raiser_pose(PreviousState_r_upper_lip_raiser_pose:double=@n22) -> (Output_Get:double)
   > Set PreviousState_mouth_emote_Ee(PreviousState_mouth_emote_Ee:double=@n14) -> (Output_Get:double)
   > Set PreviousState_mouth_emote_Oo(PreviousState_mouth_emote_Oo:double=@n16) -> (Output_Get:double)
   > Set PreviousState_mouth_emote_01(PreviousState_mouth_emote_01:double=@n8) -> (Output_Get:double)
   > Return Node()
  ]}
]
variables [
 AnimInstance {type=AnimInstance, default=null}
 PreviousState_Eye_Emote_01 {type=number}
 PreviousState_jaw_open_pose {type=number}
 PreviousState_jaw_up_pose {type=number}
 PreviousState_mouth_emote_01 {type=number}
 PreviousState_mouth_emote_Ee {type=number}
 PreviousState_mouth_emote_Oo {type=number}
 PreviousState_r_blink_pose {type=number}
 PreviousState_r_brow_in_dn_pose {type=number}
 PreviousState_r_brow_in_up_pose {type=number}
 PreviousState_r_frown_pose {type=number}
 PreviousState_r_lip_corner_wide_pose {type=number}
 PreviousState_r_lower_lid_up_pose {type=number}
 PreviousState_r_lower_lip_down_pose {type=number}
 PreviousState_r_smile_pose {type=number}
 PreviousState_r_upper_lip_raiser_pose {type=number}
 PreviousState_r_upperliddown_pose {type=number}
 PreviousState_r_wide_pose {type=number}
 PreviousState_reverse_jaw_pose {type=number}
 Threshold {type=number}
]
